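; BPI_ItemShopPDP is a Blueprint interface for components of the item shop product detail page (PDP).
; It declares callbacks for page activation and deactivation, view changes between PDP view states, and changes to item focus and interaction enablement.
; An offer-set callback passes the offer, its parent offer and the offer group as view models, leaving each implementer to supply the behavior.
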(fn OnFocusAreaChanged
  (params (IsFocusingItems :type Boolean)))

(fn OnInteractionEnabledChanged
  (params (ShouldEnable :type Boolean)))

(fn OnOfferSet
  (params (OfferVM :type Fort%20Item%20Shop%20Offer%20VM%20Object%20Reference) (ParentOfferVM :type Fort%20Item%20Shop%20Offer%20VM%20Object%20Reference) (OfferGroupVM :type Fort%20Item%20Shop%20Offer%20Group%20VM%20Object%20Reference)))

(fn OnPdpActivated)

(fn OnPdpDeactivated)

(fn OnViewChanged
  (params (NewView :type EFortItemShopPdpView%20Enum :default "OfferInspect")))
