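# Item Shop product page widget that shows a "can be purchased separately" header and View Individually button for an item within the inspected offer.
# On view change it clears its stored offer, hiding in OfferInspect and, in ItemDetails, looking up the previewed item's offer via the deprecated GetOfferForItem.
# If that offer's ID differs from the inspected one, it stores the offer, names the item and appears.
# Clicking the button broadcasts OnClicked with that offer.
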WidgetBlueprint /ItemShop/MultiProduct/ItemShopPDPComponents/WBP_ItemShopPDP_ALaCarte
graphs [
 EventGraph {nodes [
   "On Clicked (Button_ViewOffer)"!() -> (OutputDelegate, then, Button:Common Button Base Object Reference) {is=title K2Node_ComponentBoundEvent}
   > n1 = Get!(execute:Exec) -> (then:Exec, else:Exec, ALaCarteOffer:Fort Item Shop Offer VM Object Reference)
   > Call On Clicked!(execute:Exec, self:WBP Item Shop PDP ALa Carte Object Reference, ALaCarteOffer:Fort Item Shop Offer VM Object Reference=@n1) -> (then:Exec) {is=title K2Node_CallDelegate}
   n2 = Event OnViewChanged!() -> (OutputDelegate, then, NewView:EFortItemShopPdpView Enum=OfferInspect) {is=title K2Node_Event}
   > Set ALaCarteOffer(ALaCarteOffer:Fort Item Shop Offer VM Object Reference) -> (Output_Get:Fort Item Shop Offer VM Object Reference)
   > n3 = Switch on EFortItemShopPdpView!(execute, Selection:EFortItemShopPdpView Enum=@n2.NewView) -> (OfferInspect:Exec, ItemDetails:Exec) {is=title K2Node_SwitchEnum}
   n4 = Get!(execute=@n3.ItemDetails) -> (then, else, FortItemShopVM:Fort Item Shop VM Object Reference)
   > n5 = Get!(execute, self:Fort Item Shop VM Object Reference=@n4) -> (then, else, PreviewedItem:Item Shop Preview Data Object Reference)
   > n6 = Get!(execute, self:Item Shop Preview Data Object Reference=@n5) -> (then, else, ItemVM:Fort Item VM Object Reference)
   > n7 = DEPRECATED_GetOfferForItem!(
    execute
    self:Fort Item Shop Offer VM Object Reference=Get InspectedOfferVM() -> (InspectedOfferVM:Fort Item Shop Offer VM Object Reference)
    Item:Fort Item VM Object Reference=@n6
   ) -> (then, ReturnValue:Fort Item Shop Offer VM Object Reference)
   > n8 = Is Valid(InputObject=@n7)
   n9 = Get InspectedOfferVM()
   > n10 = Branch(Condition="Not Equal Exactly (String)"(A=Get OfferId!(self:Fort Item Shop Offer VM Object Reference=@n7) -> (OfferId:String), B=Get OfferId(self=@n9)))
   > Set ALaCarteOffer(ALaCarteOffer=@n7)
   > Update Item Name!(execute:Exec, ItemVM:Fort Item VM Object Reference=@n6) -> (then:Exec) {is=title K2Node_MacroInstance}
   > Update Visibility!(execute:Exec, ShouldShow:bool=true) -> (then:Exec) {is=title K2Node_MacroInstance}
   Update Visibility(execute=[@n10.else, @n8.Is Not Valid, @n6.else, @n5.else, @n4.else])
   Update Visibility(execute=@n3.OfferInspect)
   n11 = Event OnOfferSet!() -> (
    OutputDelegate
    then
    OfferVM:Fort Item Shop Offer VM Object Reference
    ParentOfferVM:Fort Item Shop Offer VM Object Reference
    OfferGroupVM:Fort Item Shop Offer Group VM Object Reference
   ) {is=title K2Node_Event}
   > Set InspectedOfferVM(InspectedOfferVM:Fort Item Shop Offer VM Object Reference=@n11.OfferVM) -> (Output_Get:Fort Item Shop Offer VM Object Reference)
  ]}
 GetDesiredFocusTarget {nodes [
   GetDesiredFocusTarget() {is=title K2Node_FunctionEntry}
   > Return Node!(execute, FocusTarget:Widget Object Reference=Get Button_ViewOffer() -> (Button_ViewOffer:WBP UIKit Button Regular Object Reference))
  ]}
 OnClicked {nodes [OnClicked!() -> (then, ALaCarteOffer:Fort Item Shop Offer VM Object Reference) {is=title K2Node_FunctionEntry}]}
 UpdateItemName {nodes [
   n1 = Inputs!() -> (execute, ItemVM:Fort Item VM Object Reference)
   n2 = Get ItemDetails!(self:Fort Item VM Object Reference=@n1) -> (ItemDetails:Fort Item Details VM Object Reference)
   > Set Text(
    Text=Format Text!(
     Format="LOCTEXT(\"<text color=\\\"FFFFFF\\\" fontface=\\\"75 medium\\\">{ItemName}</> can be purchased separately\")"
     ItemName:Text=GetOriginalDisplayName!(self:Fort Item Details VM Object Reference=@n2) -> (ReturnValue:Text)
    ) -> (Result)
    self:Rich Text Block Object Reference=Get RichText_ItemHeader() -> (RichText_ItemHeader:Common Rich Text Block Object Reference)
   )
   > Outputs!(then:Exec)
  ]}
 UpdateVisibility {nodes [
   n1 = Inputs!() -> (execute, ShouldShow:bool=~)
   > Set Visibility(
    Visibility=Select(Option 0=Collapsed, Option 1=SelfHitTestInvisible, Index=@n1)
    self=Get VerticalBox() -> (VerticalBox:Vertical Box Object Reference)
   )
   > n2 = Branch(Condition=@n1)
   Outputs(then=[@n2.then, @n3.then])
   n3 = Set Text(execute=@n2.else, self=Get RichText_ItemHeader())
  ]}
]
variables [
 ALaCarteOffer {type=FortItemShopOfferVM, default=null}
 OnClicked {}
]
widgets {
 info {inheritedWidgetCount=0, parentClass=&UserWidget, rootWidgetClass=&Overlay, widgetCount=4}
 widgets [
  Overlay_0 {widgetClassPath=&Overlay, widget=Overlay_0, properties {bOverride_Cursor=true, visibility=Visible}}
  VerticalBox {
   widgetClassPath=&VerticalBox
   bIsVariable=true
   parent=Overlay_0
   widget=VerticalBox
   slot {class=OverlaySlot, properties {padding=<ltrb 0 24 0 0>}}
  }
  RichText_ItemHeader {
   widgetClassPath=&CommonRichTextBlock
   bIsVariable=true
   parent=VerticalBox
   widget=RichText_ItemHeader
   properties {
    bOverrideDefaultStyle=true
    justification=Center
    text="[PH] <text color=\"FFFFFF\" fontface=\"75 medium\">{ItemName}</> can be purchased separately"
    wrapTextAt=610
    defaultTextStyleOverride {
     overflowPolicy=Clip
     transformPolicy=None
     colorAndOpacity {colorUseRule=UseColor_Specified, specifiedColor=<rgba 1 1 1 1>}
     font {
      bForceMonospaced=false
      bMaterialIsStencil=false
      fontMaterial=null
      fontObject=&HeadingNow
      letterSpacing=0
      monospacedWidth=1
      size=11.25
      skewAmount=0
      typefaceFontName=74 Regular
      outlineSettings {
       bApplyOutlineToDropShadows=false
       bMiteredCorners=false
       bSeparateFillAlpha=false
       outlineMaterial=null
       outlineSize=0
       outlineColor=<rgba 0 0 0 1>
      }
     }
     highlightColor {colorUseRule=UseColor_Specified, specifiedColor=<rgba 0 0 0 1>}
     highlightShape {
      bIsDynamicallyLoaded=false
      drawAs=Image
      imageType=NoImage
      mirroring=NoMirror
      resourceName=None
      resourceObject=null
      tiling=NoTile
      imageSize=<32 32>
      margin=<ltrb 0 0 0 0>
      outlineSettings {
       bUseBrushTransparency=false
       roundingType=HalfHeightRadius
       width=0
       color {colorUseRule=UseColor_Specified, specifiedColor=<rgba 0 0 0 0>}
       cornerRadii=<0 0 0 0>
      }
      tintColor {colorUseRule=UseColor_Specified, specifiedColor=<rgba 1 1 1 1>}
      uVRegion {bIsValid=false, max=<0 0>, min=<0 0>}
     }
     selectedBackgroundColor {colorUseRule=UseColor_Foreground, specifiedColor=<rgba 1 0 1 1>}
     shadowColorAndOpacity=<rgba 0 0 0 1>
     shadowOffset=<0 0>
     strikeBrush {
      bIsDynamicallyLoaded=false
      drawAs=Image
      imageType=NoImage
      mirroring=NoMirror
      resourceName=None
      resourceObject=null
      tiling=NoTile
      imageSize=<32 32>
      margin=<ltrb 0 0 0 0>
      outlineSettings {
       bUseBrushTransparency=false
       roundingType=HalfHeightRadius
       width=0
       color {colorUseRule=UseColor_Specified, specifiedColor=<rgba 0 0 0 0>}
       cornerRadii=<0 0 0 0>
      }
      tintColor {colorUseRule=UseColor_Specified, specifiedColor=<rgba 1 1 1 1>}
      uVRegion {bIsValid=false, max=<0 0>, min=<0 0>}
     }
     underlineBrush {
      bIsDynamicallyLoaded=false
      drawAs=Image
      imageType=NoImage
      mirroring=NoMirror
      resourceName=None
      resourceObject=null
      tiling=NoTile
      imageSize=<32 32>
      margin=<ltrb 0 0 0 0>
      outlineSettings {
       bUseBrushTransparency=false
       roundingType=HalfHeightRadius
       width=0
       color {colorUseRule=UseColor_Specified, specifiedColor=<rgba 0 0 0 0>}
       cornerRadii=<0 0 0 0>
      }
      tintColor {colorUseRule=UseColor_Specified, specifiedColor=<rgba 1 1 1 1>}
      uVRegion {bIsValid=false, max=<0 0>, min=<0 0>}
     }
    }
   }
   slot {class=VerticalBoxSlot, properties {padding=<ltrb 0 0 0 8>}}
  }
  Button_ViewOffer {
   widgetClassPath=&WBP_UIKit_Button_Regular
   bIsVariable=true
   parent=VerticalBox
   widget=Button_ViewOffer
   properties {bShouldUseFallbackDefaultInputAction=false, buttonSize=Small, minWidthOverride=610, mobileMinWidth=610, text=View Individually}
   slot {class=VerticalBoxSlot}
  }
 ]
}
