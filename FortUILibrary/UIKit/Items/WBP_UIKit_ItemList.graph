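# Reusable UI Kit horizontal list of item cards with fade edges and scroll arrows.
# SetItemList builds cards in a dynamic entry box and relays hover, click and select through dispatchers with index and item.
# Arrows step the selected index by ColumnsToScroll, flipping for right-to-left cultures when IsFlowDirectionCulture is set.
# Overflow and scroll offset drive fade and arrow visibility; gamepad hides arrows, and touch reselects the current card.
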WidgetBlueprint /FortUILibrary/UIKit/Items/WBP_UIKit_ItemList
graphs [
 CreateItemCardBindings {nodes [
   n1 = CreateItemCardBindings!() -> (then, Item Card:WBP UIKit Item Card Base Object Reference) {is=title K2Node_FunctionEntry}
   n2 = Create Event()
   > Unbind Event from On Hovered(self:Common Button Base Object Reference=@n1, Delegate=@n2) {is=title K2Node_RemoveDelegate}
   > Bind Event to On Hovered(self:Common Button Base Object Reference=@n1, Delegate=@n2) {is=title K2Node_AddDelegate}
   n3 = Create Event()
   > Unbind Event from On Clicked(self:Common Button Base Object Reference=@n1, Delegate=@n3) {is=title K2Node_RemoveDelegate}
   > Bind Event to On Clicked(self:Common Button Base Object Reference=@n1, Delegate=@n3) {is=title K2Node_AddDelegate}
   n4 = Create Event()
   > Unbind Event from On Selected(self:Common Button Base Object Reference=@n1, Delegate=@n4) {is=title K2Node_RemoveDelegate}
   > Bind Event to On Selected(self:Common Button Base Object Reference=@n1, Delegate=@n4) {is=title K2Node_AddDelegate}
  ]}
 EventGraph {nodes [
   "On User Scrolled (Scrollbox_Items)"!() -> (OutputDelegate, then, CurrentOffset:float) {is=title K2Node_ComponentBoundEvent}
   > UpdateScrollboxPresentation()
   "On Clicked (ScrollArrow_Right)"!() -> (OutputDelegate, then, Button:Common Button Base Object Reference) {is=title K2Node_ComponentBoundEvent}
   > ScrollList!(execute, self, IsScrollingRight:bool=true) -> (then)
   "On Clicked (ScrollArrow_Left)"!() -> (OutputDelegate, then, Button:Common Button Base Object Reference) {is=title K2Node_ComponentBoundEvent}
   > ScrollList(IsScrollingRight=false)
   n1 = OnItemHovered!() -> (OutputDelegate, then, Button:Common Button Base Object Reference) {is=title K2Node_CustomEvent}
   n2 = GetItemFromWidget!(self, InWidget:"User Widget Object Reference (by ref)"=@n1.Button) -> (OutIndex:int=0, OutItem:Object Reference)
   > Call Item Hovered!(
    execute:Exec
    self:WBP UIKit Item List Object Reference
    Button:Common Button Base Object Reference=@n1.Button
    ItemIndex:int=@n2
    ListItem:Object Reference=@n2.OutItem
   ) -> (then:Exec) {is=title K2Node_CallDelegate}
   n3 = OnItemClicked!() -> (OutputDelegate, then, Button:Common Button Base Object Reference) {is=title K2Node_CustomEvent}
   n4 = GetItemFromWidget(InWidget=@n3.Button) -> (OutIndex=0)
   > n5 = Set SelectedCardIndex(SelectedCardIndex:int=@n4) -> (Output_Get:int=0)
   > Call Item Clicked!(
    execute:Exec
    self:WBP UIKit Item List Object Reference
    Button:Common Button Base Object Reference=@n3.Button
    ItemIndex:int=@n5
    ListItem:Object Reference=@n4.OutItem
   ) -> (then:Exec) {is=title K2Node_CallDelegate}
   Activation() {is=title K2Node_CustomEvent}
   > AddWidgets!(
    execute
    self:Common Widget Group Base Object Reference=Get ItemCardsButtonGroup() -> (ItemCardsButtonGroup:Common Button Group Base Object Reference)
    Widgets:Array of Widget Object References=GetAllEntries!(self:Dynamic Entry Box Base Object Reference=Get FortDynamicEntryBox_Items() -> (FortDynamicEntryBox_Items:Fort Dynamic Entry Box Object Reference)) -> (ReturnValue:Array of User Widget Object References)
   ) -> (then)
   > n6 = Branch(Condition=IsInputType!(InputType:ECommonInputType Enum=Touch, IsNot:bool=false) -> (Result:bool))
   n7 = GetSelectedWidget!(self) -> (OutWidget:Widget Object Reference)
   Is Valid(exec=[@n11.then, @n6.then], InputObject=@n7)
   > n8 = Cast To WBP_UIKit_ItemList_CardBase!(execute, Object=@n7) -> (then, CastFailed, AsWBP UIKit Item List Card Base:WBP UIKit Item List Card Base Object Reference) {is=title K2Node_DynamicCast}
   > SetIsSelected!(execute, self:Common Button Base Object Reference=@n8, InSelected:bool=true, bGiveClickFeedback:bool=false) -> (then)
   Deactivation() {is=title K2Node_CustomEvent}
   n9 = Get ItemCardsButtonGroup()
   > DeselectAll(self:Common Button Group Base Object Reference=@n9)
   > RemoveAll(self:Common Widget Group Base Object Reference=@n9)
   n10 = OnInputMethodChanged!() -> (OutputDelegate, then, bNewInputType:ECommonInputType Enum) {is=title K2Node_CustomEvent}
   > DeselectAll(self=Get ItemCardsButtonGroup())
   > UpdateScrollboxPresentation()
   > n11 = Branch(Condition="Equal (Enum)"(A=@n10.bNewInputType, B=Touch))
   n12 = OnItemSelected!() -> (OutputDelegate, then, Button:Common Button Base Object Reference) {is=title K2Node_CustomEvent}
   n13 = GetItemFromWidget(InWidget=@n12.Button) -> (OutIndex=0)
   > n14 = Set SelectedCardIndex(SelectedCardIndex=@n13) -> (Output_Get=0)
   > Call Item Selected!(
    execute:Exec
    self:WBP UIKit Item List Object Reference
    Button:Common Button Base Object Reference=@n12.Button
    ItemIndex:int=@n14
    ListItem:Object Reference=@n13.OutItem
   ) -> (then:Exec) {is=title K2Node_CallDelegate}
   Event OnInitialized()
   > Bind Event to On Input Method Changed(self=Get CommonInputSubsystem(), Delegate=Create Event())
   n15 = Event PreConstruct()
   > n16 = Construct Object from Class!(execute, Class=/Script/CommonUI.CommonButtonGroupBase, self, bSelectionRequired:bool=False) -> (then, ReturnValue:Common Button Group Base Object Reference)
   > Set ItemCardsButtonGroup(ItemCardsButtonGroup:Common Button Group Base Object Reference=@n16) -> (Output_Get:Common Button Group Base Object Reference)
   > Set WidthOverride!(
    execute
    WidthOverride:float=Get ListWidth() -> (ListWidth:double)
    self:Size Box Object Reference=Get SizeBox_Items() -> (SizeBox_Items:Size Box Object Reference)
   ) -> (then, Output_Get:float)
   > SetCompactStyle()
   > Set EntrySpacing!(
    execute
    EntrySpacing:Vector 2D Structure=Get Entry Spacing() -> (Entry Spacing:Vector 2D Structure)
    self:Dynamic Entry Box Base Object Reference=Get FortDynamicEntryBox_Items()
   ) -> (then, Output_Get:Vector 2D Structure)
   > SetHorizontalAlignment(
    self:Scroll Box Slot Object Reference=SlotAsScrollBoxSlot!(Widget:Widget Object Reference=Get FortDynamicEntryBox_Items()) -> (ReturnValue:Scroll Box Slot Object Reference)
    InHorizontalAlignment=Get HorizontalAlignment() -> (HorizontalAlignment:EHorizontalAlignment Enum=HAlign_Fill)
   )
   > Reset!(execute, self:Dynamic Entry Box Object Reference=Get FortDynamicEntryBox_Items(), bDeleteWidgets:bool=true) -> (then)
   > Branch(Condition=@n15.IsDesignTime)
   > For Loop(FirstIndex=1, LastIndex=Get DesignerPreviewEntries() -> (DesignerPreviewEntries:int=0))
   > Create Entry of Class!(
    execute
    self:Dynamic Entry Box Object Reference=Get FortDynamicEntryBox_Items()
    EntryClass:User Widget Class Reference=Get Entry Class() -> (Entry Class:WBP UIKit Item List Card Base Class Reference)
   ) -> (then, ReturnValue:WBP UIKit Item List Card Base Object Reference)
  ]}
 GetItemFromWidget {nodes [
   GetItemFromWidget!() -> (then, InWidget:"User Widget Object Reference (by ref)") {is=title K2Node_FunctionEntry}
   n1 = Get FortDynamicEntryBox_Items()
   > n2 = Set FoundIndex(FoundIndex:int=Find Item!(
     TargetArray:Array of User Widget Object References=GetAllEntries(self=@n1)
     ItemToFind:"User Widget Object Reference (by ref)"=Get InWidget() -> (InWidget:"User Widget Object Reference (by ref)")
    ) -> (ReturnValue)) -> (Output_Get:int=0)
   > Return Node!(
    execute
    OutIndex:int=@n2
    OutItem:Object Reference="Get (a copy)"(Array:Array of Object References=Get AllItems() -> (AllItems:Array of Object References), Dimension 1=@n2) -> (Output:Object Reference)
   )
  ]}
 GetNextIndex {nodes [
   Inputs!()
   n1 = Get ColumnsToScroll() -> (ColumnsToScroll:int=0)
   n2 = Get SelectedCardIndex() -> (SelectedCardIndex:int=0)
   Outputs!(ReturnValue:int="Min (Integer)"(A=int + int(A=@n2, B=@n1), B=Get LastCardIndex() -> (LastCardIndex:int=0)) -> (ReturnValue=0))
  ]}
 GetPreviousIndex {nodes [
   Inputs()
   n1 = Get SelectedCardIndex() -> (SelectedCardIndex=0)
   n2 = Get ColumnsToScroll() -> (ColumnsToScroll=0)
   Outputs(ReturnValue="Max (Integer)"(A=int - int(A=@n1, B=@n2), B=0))
  ]}
 GetSelectedWidget {nodes [
   GetSelectedWidget() {is=title K2Node_FunctionEntry}
   > Return Node!(
    execute
    OutWidget:Widget Object Reference=GetWidgetAtIndex!(self, InIndex:int=Get SelectedCardIndex() -> (SelectedCardIndex=0)) -> (OutWIdget:Widget Object Reference)
   )
  ]}
 GetWidgetAtIndex {nodes [
   n1 = GetWidgetAtIndex!() -> (then, InIndex:int) {is=title K2Node_FunctionEntry}
   n2 = Get FortDynamicEntryBox_Items()
   > Return Node!(
    execute
    OutWIdget:Widget Object Reference="Get (a copy)"(Array:Array of User Widget Object References=GetAllEntries(self=@n2), Dimension 1=@n1) -> (Output:User Widget Object Reference)
   )
  ]}
 IsLTRCulture {nodes [
   IsLTRCulture() {is=title K2Node_FunctionEntry}
   > n1 = Branch(Condition=Get IsFlowDirectionCulture() -> (IsFlowDirectionCulture:bool))
   n2 = GetCurrentCulture!() -> (ReturnValue:String)
   > Return Node!(execute, IsLTRCulture:bool=NOT Boolean(A=IsCultureRightToLeft!(Culture:String=@n2) -> (ReturnValue:bool)))
   Return Node(execute=@n1.else, IsLTRCulture=true)
  ]}
 ItemClicked {nodes [ItemClicked!() -> (then, Button:Common Button Base Object Reference, ItemIndex:int, ListItem:Object Reference) {is=title K2Node_FunctionEntry}]}
 ItemHovered {nodes [ItemHovered!() -> (then, Button:Common Button Base Object Reference, ItemIndex:int, ListItem:Object Reference) {is=title K2Node_FunctionEntry}]}
 ItemSelected {nodes [ItemSelected!() -> (then, Button:Common Button Base Object Reference, ItemIndex:int, ListItem:Object Reference) {is=title K2Node_FunctionEntry}]}
 ScrollList {nodes [
   ScrollList!() -> (then, IsScrollingRight:bool=~) {is=title K2Node_FunctionEntry}
   n1 = Get Previous Index!() -> (ReturnValue:int) {is=title K2Node_MacroInstance}
   n2 = Get Next Index!() -> (ReturnValue:int) {is=title K2Node_MacroInstance}
   n3 = Get IsScrollingRight() -> (IsScrollingRight:bool)
   n4 = Get Previous Index()
   n5 = Get Next Index()
   n6 = Get IsScrollingRight()
   > n7 = Set SelectedCardIndex(SelectedCardIndex=Select(
     Option 0:int=Select(Option 0=@n5, Option 1=@n4, Index=@n6)
     Option 1:int=Select(Option 0=@n1, Option 1=@n2, Index=@n3)
     Index=IsLTRCulture!() -> (IsLTRCulture:bool) {is=title K2Node_CallFunction}
    ) -> (ReturnValue:int)) -> (Output_Get=0)
   n8 = Get FortDynamicEntryBox_Items()
   > ScrollToCard!(execute, self, WidgetToFind:Widget Object Reference="Get (a copy)"(Array=GetAllEntries(self=@n8), Dimension 1=@n7), AnimateScroll:bool=true) -> (then)
  ]}
 ScrollToCard {nodes [
   ScrollToCard!() -> (then, WidgetToFind:Widget Object Reference, AnimateScroll:bool=~) {is=title K2Node_FunctionEntry}
   > ScrollWidgetIntoView!(
    execute
    self:Scroll Box Object Reference=Get Scrollbox_Items() -> (Scrollbox_Items:Athena Scroll Box Object Reference)
    WidgetToFind:Widget Object Reference=Get WidgetToFind() -> (WidgetToFind:Widget Object Reference)
    AnimateScroll:bool=Get AnimateScroll() -> (AnimateScroll:bool)
    ScrollDestination:EDescendantScrollDestination Enum=Select(Option 0:EDescendantScrollDestination Enum=BottomOrRight, Option 1:EDescendantScrollDestination Enum=TopOrLeft, Index=IsLTRCulture()) -> (ReturnValue:EDescendantScrollDestination Enum)
    Padding:float=Get NavigationScrollPadding!(self:Scroll Box Object Reference=Get Scrollbox_Items()) -> (NavigationScrollPadding:float)
   ) -> (then)
  ]}
 SetCompactStyle {nodes [
   SetCompactStyle() {is=title K2Node_FunctionEntry}
   > Branch(Condition=IsUsingCompactUI())
   > SetWidthOverride(self=Get SizeBox_Items(), InWidthOverride=Get CompactListWidth() -> (CompactListWidth:double))
  ]}
 SetItemList {nodes [
   n1 = SetItemList!() -> (then, ItemList:Array of Object References) {is=title K2Node_FunctionEntry}
   > Set AllItems(AllItems:Array of Object References=@n1) -> (Output_Get:Array of Object References)
   > Reset(self=Get FortDynamicEntryBox_Items(), bDeleteWidgets=false)
   > n2 = For Each Loop(Array:Array of Object References=Get AllItems()) -> (Array Element:Object Reference)
   > Set LoopItem(LoopItem:Object Reference=@n2) -> (Output_Get:Object Reference)
   > n3 = Create Entry of Class(self=Get FortDynamicEntryBox_Items(), EntryClass=Get Entry Class())
   > n4 = Set LoopCard(LoopCard:WBP UIKit Item List Card Base Object Reference=@n3) -> (Output_Get:WBP UIKit Item List Card Base Object Reference)
   > Set ItemCardSize(ItemCardSize:Vector 2D Structure=Get ItemCardSize!(self:WBP UIKit Item Card Base Object Reference=@n4) -> (ItemCardSize:Vector 2D Structure)) -> (Output_Get:Vector 2D Structure)
   > InitializeItemCard(self:WBP UIKit Item Card Base Object Reference=Get LoopCard() -> (LoopCard:WBP UIKit Item List Card Base Object Reference))
   > CreateItemCardBindings!(execute, self, Item Card:WBP UIKit Item Card Base Object Reference=Get LoopCard()) -> (then) {is=title K2Node_CallFunction}
   > SetListItem!(
    execute
    self:WBP UIKit Item List Card Base Object Reference=Get LoopCard()
    InListItem:Object Reference=Get LoopItem() -> (LoopItem:Object Reference)
   ) -> (then)
   Set LastCardIndex(execute=@n2.Completed, LastCardIndex:int=@n2.Array Index) -> (Output_Get:int=0)
   > Set SelectedCardIndex(SelectedCardIndex=0) -> (Output_Get=0)
   n5 = Get FortDynamicEntryBox_Items()
   > ScrollToCard(WidgetToFind="Get (a copy)"(Array=GetAllEntries(self=@n5)), AnimateScroll=false) {is=title K2Node_CallFunction}
   > UpdateScrollboxPresentation()
  ]}
 UpdatePresentations {nodes [
   UpdatePresentations() {is=title K2Node_FunctionEntry}
   > n1 = For Each Loop(Array:Array of User Widget Object References=GetAllEntries(self=Get FortDynamicEntryBox_Items())) -> (Array Element:User Widget Object Reference)
   > n2 = Cast To WBP_UIKit_ItemList_CardBase(Object=@n1)
   > UpdatePresentation(self:WBP UIKit Item List Card Base Object Reference=@n2)
  ]}
 UpdateScrollboxPresentation {nodes [
   UpdateScrollboxPresentation() {is=title K2Node_FunctionEntry}
   n1 = Get FortDynamicEntryBox_Items()
   n2 = GetNumEntries!(self:Dynamic Entry Box Base Object Reference=@n1) -> (ReturnValue:int=0)
   n3 = float * float(A:int=int - int(A=@n2), B=Get!(self:Dynamic Entry Box Base Object Reference=@n1) -> (EntrySpacing_X:double, EntrySpacing_Y:double)) {is=title K2Node_PromotableOperator}
   n4 = float * float(A=@n2, B=Get!() -> (ItemCardSize_X:double, ItemCardSize_Y:double))
   n5 = Get SizeBox_Items()
   > n6 = Branch(Condition="float <= float"(A=float + float(A=@n4, B=@n3), B=Get WidthOverride!(self:Size Box Object Reference=@n5) -> (WidthOverride:float)) -> (ReturnValue=~))
   > SetVisibility(
    self=[
     Get Spacer_StartPadding() -> (Spacer_StartPadding:Spacer Object Reference)
     Get ScrollArrow_Left() -> (ScrollArrow_Left:WBP UIKit Button Quiet Object Reference)
     Get ScrollArrow_Right() -> (ScrollArrow_Right:WBP UIKit Button Quiet Object Reference)
     Get Image_Fade_Left() -> (Image_Fade_Left:Image Object Reference)
     Get Image_Fade_Right() -> (Image_Fade_Right:Image Object Reference)
    ]
    InVisibility=Collapsed
   )
   Set Visibility(execute=@n6.else, Visibility=SelfHitTestInvisible, self=Get Spacer_StartPadding())
   > ForceLayoutPrepass(self:Widget Object Reference=Get Scrollbox_Items())
   n7 = Get Scrollbox_Items()
   n8 = GetScrollOffset!(self:Scroll Box Object Reference=@n7) -> (ReturnValue:float)
   n9 = Get!(self:Spacer Object Reference=Get Spacer_StartPadding()) -> (Size_X:double, Size_Y:double)
   > Set CanNavigateLeft(CanNavigateLeft:bool=NOT Boolean(A="Nearly Equal (Float)"(A=@n8, B=0, ErrorTolerance=@n9))) -> (Output_Get:bool)
   n10 = GetScrollOffsetOfEnd!(self:Scroll Box Object Reference=@n7) -> (ReturnValue:float)
   > Set CanNavigateRight(CanNavigateRight:bool=NOT Boolean(A="Nearly Equal (Float)"(A=@n8, B=@n10, ErrorTolerance=@n9))) -> (Output_Get:bool)
   > SetVisibility(
    self=Get Image_Fade_Left()
    InVisibility=Select(
     Option 0:ESlateVisibility Enum=Collapsed
     Option 1:ESlateVisibility Enum=SelfHitTestInvisible
     Index=Get CanNavigateLeft() -> (CanNavigateLeft:bool)
    ) -> (ReturnValue:ESlateVisibility Enum)
   )
   > SetVisibility(
    self=Get Image_Fade_Right()
    InVisibility=Select(Option 0=Collapsed, Option 1=SelfHitTestInvisible, Index=Get CanNavigateRight() -> (CanNavigateRight:bool))
   )
   > n11 = Branch(Condition=IsInputType(InputType=Gamepad, IsNot=false))
   > SetVisibility(self=[Get ScrollArrow_Left(), Get ScrollArrow_Right()], InVisibility=Collapsed)
   SetVisibility(
    execute=@n11.else
    self=Get ScrollArrow_Left()
    InVisibility=Select(Option 0=Collapsed, Option 1=SelfHitTestInvisible, Index=Get CanNavigateLeft())
   )
   > SetVisibility(self=Get ScrollArrow_Right(), InVisibility=Select(Option 0=Collapsed, Option 1=SelfHitTestInvisible, Index=Get CanNavigateRight()))
  ]}
]
variables [
 AllItems {type="array<Object>", default []}
 ColumnsToScroll {type=integer, default=5}
 CompactListWidth {type=number, default=789}
 DesignerPreviewEntries {type=integer, default=7}
 Entry Class {type=Class, default=/FortUILibrary/Systems/Items/WBP_UIKit_ItemList_CardBase.WBP_UIKit_ItemList_CardBase_C}
 Entry Spacing {type=Vector2D, default=<16 0>}
 HorizontalAlignment {type=EHorizontalAlignment, default=HAlign_Left}
 IsFlowDirectionCulture {type=boolean}
 ItemCardsButtonGroup {type=CommonButtonGroupBase, default=null}
 ItemCardSize {type=Vector2D, default=<0 0>}
 ItemClicked {}
 ItemHovered {}
 ItemSelected {}
 LastCardIndex {type=integer}
 ListWidth {type=number, default=732}
 SelectedCardIndex {type=integer}
]
widgets {
 info {inheritedWidgetCount=0, parentClass=&UserWidget, rootWidgetClass=&GridPanel, widgetCount=12}
 widgets [
  Overlay_RowContent {widgetClassPath=&GridPanel, widget=Overlay_RowContent, properties {bOverride_Cursor=true}}
  PostBufferUpdate {
   widgetClassPath=/Script/UMG.PostBufferUpdate
   parent=Overlay_RowContent
   widget=PostBufferUpdate
   properties {visibility=HitTestInvisible, updateBufferInfos [{bufferToUpdate=ESlatePostRT_1, postParamUpdater=null}]}
   slot {class=GridSlot}
  }
  Overlay_Clipper {
   widgetClassPath=&Overlay
   parent=Overlay_RowContent
   widget=Overlay_Clipper
   properties {clipping=ClipToBounds}
   slot {class=GridSlot}
  }
  SizeBox_Items {
   widgetClassPath=&SizeBox
   bIsVariable=true
   parent=Overlay_Clipper
   widget=SizeBox_Items
   properties {bOverride_WidthOverride=true, heightOverride=144, widthOverride=732}
   slot {class=OverlaySlot, properties {horizontalAlignment=HAlign_Center, verticalAlignment=VAlign_Center, padding=<ltrb 24 0 24 0>}}
  }
  Scrollbox_Items {
   widgetClassPath=/Script/FortniteUI.AthenaScrollBox
   bIsVariable=true
   parent=SizeBox_Items
   widget=Scrollbox_Items
   properties {
    allowOverscroll=false
    clipping=Inherit
    consumeMouseWheel=Never
    orientation=Orient_Horizontal
    scrollBarVisibility=Collapsed
    scrollWhenFocusChanges=AnimatedScroll
    wheelScrollMultiplier=0
    scrollbarPadding=<ltrb 0 0 0 0>
    scrollbarThickness=<0 0>
    widgetBarStyle {
     thickness=0
     draggedThumbImage {
      bIsDynamicallyLoaded=false
      drawAs=Box
      imageType=FullColor
      mirroring=NoMirror
      resourceName=../../../Engine/Content/Slate/Common/Scrollbar_Thumb.png
      resourceObject=null
      tiling=NoTile
      imageSize=<0 0>
      margin=<ltrb 0.25 0.25 0.25 0.25>
      outlineSettings {
       bUseBrushTransparency=false
       roundingType=HalfHeightRadius
       width=0
       color {colorUseRule=UseColor_Specified, specifiedColor=<rgba 0 0 0 0>}
       cornerRadii=<0 0 0 0>
      }
      tintColor {colorUseRule=UseColor_Specified, specifiedColor=<rgba 1 1 1 1>}
      uVRegion {bIsValid=false, max=<0 0>, min=<0 0>}
     }
     horizontalBackgroundImage {
      bIsDynamicallyLoaded=false
      drawAs=NoDrawType
      imageType=NoImage
      mirroring=NoMirror
      resourceName=None
      resourceObject=null
      tiling=NoTile
      imageSize=<0 0>
      margin=<ltrb 0 0 0 0>
      outlineSettings {
       bUseBrushTransparency=false
       roundingType=HalfHeightRadius
       width=0
       color {colorUseRule=UseColor_Specified, specifiedColor=<rgba 0 0 0 0>}
       cornerRadii=<0 0 0 0>
      }
      tintColor {colorUseRule=UseColor_Specified, specifiedColor=<rgba 1 1 1 1>}
      uVRegion {bIsValid=false, max=<0 0>, min=<0 0>}
     }
     horizontalBottomSlotImage {
      bIsDynamicallyLoaded=false
      drawAs=Image
      imageType=FullColor
      mirroring=NoMirror
      resourceName=../../../Engine/Content/Slate/Common/Scrollbar_Background_Horizontal.png
      resourceObject=null
      tiling=NoTile
      imageSize=<8 8>
      margin=<ltrb 0 0 0 0>
      outlineSettings {
       bUseBrushTransparency=false
       roundingType=HalfHeightRadius
       width=0
       color {colorUseRule=UseColor_Specified, specifiedColor=<rgba 0 0 0 0>}
       cornerRadii=<0 0 0 0>
      }
      tintColor {colorUseRule=UseColor_Specified, specifiedColor=<rgba 1 1 1 1>}
      uVRegion {bIsValid=false, max=<0 0>, min=<0 0>}
     }
     horizontalTopSlotImage {
      bIsDynamicallyLoaded=false
      drawAs=Image
      imageType=FullColor
      mirroring=NoMirror
      resourceName=../../../Engine/Content/Slate/Common/Scrollbar_Background_Horizontal.png
      resourceObject=null
      tiling=NoTile
      imageSize=<8 8>
      margin=<ltrb 0 0 0 0>
      outlineSettings {
       bUseBrushTransparency=false
       roundingType=HalfHeightRadius
       width=0
       color {colorUseRule=UseColor_Specified, specifiedColor=<rgba 0 0 0 0>}
       cornerRadii=<0 0 0 0>
      }
      tintColor {colorUseRule=UseColor_Specified, specifiedColor=<rgba 1 1 1 1>}
      uVRegion {bIsValid=false, max=<0 0>, min=<0 0>}
     }
     hoveredThumbImage {
      bIsDynamicallyLoaded=false
      drawAs=Box
      imageType=FullColor
      mirroring=NoMirror
      resourceName=../../../Engine/Content/Slate/Common/Scrollbar_Thumb.png
      resourceObject=null
      tiling=NoTile
      imageSize=<0 0>
      margin=<ltrb 0.25 0.25 0.25 0.25>
      outlineSettings {
       bUseBrushTransparency=false
       roundingType=HalfHeightRadius
       width=0
       color {colorUseRule=UseColor_Specified, specifiedColor=<rgba 0 0 0 0>}
       cornerRadii=<0 0 0 0>
      }
      tintColor {colorUseRule=UseColor_Specified, specifiedColor=<rgba 1 1 1 1>}
      uVRegion {bIsValid=false, max=<0 0>, min=<0 0>}
     }
     normalThumbImage {
      bIsDynamicallyLoaded=false
      drawAs=Box
      imageType=FullColor
      mirroring=NoMirror
      resourceName=../../../Engine/Content/Slate/Common/Scrollbar_Thumb.png
      resourceObject=null
      tiling=NoTile
      imageSize=<0 0>
      margin=<ltrb 0.25 0.25 0.25 0.25>
      outlineSettings {
       bUseBrushTransparency=false
       roundingType=HalfHeightRadius
       width=0
       color {colorUseRule=UseColor_Specified, specifiedColor=<rgba 0 0 0 0>}
       cornerRadii=<0 0 0 0>
      }
      tintColor {colorUseRule=UseColor_Specified, specifiedColor=<rgba 1 1 1 1>}
      uVRegion {bIsValid=false, max=<0 0>, min=<0 0>}
     }
     verticalBackgroundImage {
      bIsDynamicallyLoaded=false
      drawAs=NoDrawType
      imageType=NoImage
      mirroring=NoMirror
      resourceName=None
      resourceObject=null
      tiling=NoTile
      imageSize=<0 0>
      margin=<ltrb 0 0 0 0>
      outlineSettings {
       bUseBrushTransparency=false
       roundingType=HalfHeightRadius
       width=0
       color {colorUseRule=UseColor_Specified, specifiedColor=<rgba 0 0 0 0>}
       cornerRadii=<0 0 0 0>
      }
      tintColor {colorUseRule=UseColor_Specified, specifiedColor=<rgba 1 1 1 1>}
      uVRegion {bIsValid=false, max=<0 0>, min=<0 0>}
     }
     verticalBottomSlotImage {
      bIsDynamicallyLoaded=false
      drawAs=Image
      imageType=FullColor
      mirroring=NoMirror
      resourceName=../../../Engine/Content/Slate/Common/Scrollbar_Background_Vertical.png
      resourceObject=null
      tiling=NoTile
      imageSize=<8 8>
      margin=<ltrb 0 0 0 0>
      outlineSettings {
       bUseBrushTransparency=false
       roundingType=HalfHeightRadius
       width=0
       color {colorUseRule=UseColor_Specified, specifiedColor=<rgba 0 0 0 0>}
       cornerRadii=<0 0 0 0>
      }
      tintColor {colorUseRule=UseColor_Specified, specifiedColor=<rgba 1 1 1 1>}
      uVRegion {bIsValid=false, max=<0 0>, min=<0 0>}
     }
     verticalTopSlotImage {
      bIsDynamicallyLoaded=false
      drawAs=Image
      imageType=FullColor
      mirroring=NoMirror
      resourceName=../../../Engine/Content/Slate/Common/Scrollbar_Background_Vertical.png
      resourceObject=null
      tiling=NoTile
      imageSize=<8 8>
      margin=<ltrb 0 0 0 0>
      outlineSettings {
       bUseBrushTransparency=false
       roundingType=HalfHeightRadius
       width=0
       color {colorUseRule=UseColor_Specified, specifiedColor=<rgba 0 0 0 0>}
       cornerRadii=<0 0 0 0>
      }
      tintColor {colorUseRule=UseColor_Specified, specifiedColor=<rgba 1 1 1 1>}
      uVRegion {bIsValid=false, max=<0 0>, min=<0 0>}
     }
    }
    widgetStyle {
     barThickness=9
     bottomShadowBrush {
      bIsDynamicallyLoaded=false
      drawAs=NoDrawType
      imageType=FullColor
      mirroring=NoMirror
      resourceName=../../../Engine/Content/Slate/Common/ScrollBoxShadowBottom.png
      resourceObject=null
      tiling=NoTile
      imageSize=<16 8>
      margin=<ltrb 0.5 0 0.5 1>
      outlineSettings {
       bUseBrushTransparency=false
       roundingType=HalfHeightRadius
       width=0
       color {colorUseRule=UseColor_Specified, specifiedColor=<rgba 0 0 0 0>}
       cornerRadii=<0 0 0 0>
      }
      tintColor {colorUseRule=UseColor_Specified, specifiedColor=<rgba 1 1 1 1>}
      uVRegion {bIsValid=false, max=<0 0>, min=<0 0>}
     }
     horizontalScrolledContentPadding=<ltrb 0 0 0 0>
     leftShadowBrush {
      bIsDynamicallyLoaded=false
      drawAs=NoDrawType
      imageType=FullColor
      mirroring=NoMirror
      resourceName=../../../Engine/Content/Slate/Common/ScrollBoxShadowLeft.png
      resourceObject=null
      tiling=NoTile
      imageSize=<8 16>
      margin=<ltrb 1 0.5 0 0.5>
      outlineSettings {
       bUseBrushTransparency=false
       roundingType=HalfHeightRadius
       width=0
       color {colorUseRule=UseColor_Specified, specifiedColor=<rgba 0 0 0 0>}
       cornerRadii=<0 0 0 0>
      }
      tintColor {colorUseRule=UseColor_Specified, specifiedColor=<rgba 1 1 1 1>}
      uVRegion {bIsValid=false, max=<0 0>, min=<0 0>}
     }
     rightShadowBrush {
      bIsDynamicallyLoaded=false
      drawAs=NoDrawType
      imageType=FullColor
      mirroring=NoMirror
      resourceName=../../../Engine/Content/Slate/Common/ScrollBoxShadowRight.png
      resourceObject=null
      tiling=NoTile
      imageSize=<8 16>
      margin=<ltrb 0 0.5 1 0.5>
      outlineSettings {
       bUseBrushTransparency=false
       roundingType=HalfHeightRadius
       width=0
       color {colorUseRule=UseColor_Specified, specifiedColor=<rgba 0 0 0 0>}
       cornerRadii=<0 0 0 0>
      }
      tintColor {colorUseRule=UseColor_Specified, specifiedColor=<rgba 1 1 1 1>}
      uVRegion {bIsValid=false, max=<0 0>, min=<0 0>}
     }
     topShadowBrush {
      bIsDynamicallyLoaded=false
      drawAs=NoDrawType
      imageType=FullColor
      mirroring=NoMirror
      resourceName=../../../Engine/Content/Slate/Common/ScrollBoxShadowTop.png
      resourceObject=null
      tiling=NoTile
      imageSize=<16 8>
      margin=<ltrb 0.5 1 0.5 0>
      outlineSettings {
       bUseBrushTransparency=false
       roundingType=HalfHeightRadius
       width=0
       color {colorUseRule=UseColor_Specified, specifiedColor=<rgba 0 0 0 0>}
       cornerRadii=<0 0 0 0>
      }
      tintColor {colorUseRule=UseColor_Specified, specifiedColor=<rgba 1 1 1 1>}
      uVRegion {bIsValid=false, max=<0 0>, min=<0 0>}
     }
     verticalScrolledContentPadding=<ltrb 0 0 0 0>
    }
   }
   slot {class=SizeBoxSlot}
  }
  Spacer_StartPadding {
   widgetClassPath=&Spacer
   bIsVariable=true
   parent=Scrollbox_Items
   widget=Spacer_StartPadding
   properties {size=<24 1>}
   slot {class=ScrollBoxSlot}
  }
  FortDynamicEntryBox_Items {
   widgetClassPath=/Script/FortniteUI.FortDynamicEntryBox
   bIsVariable=true
   parent=Scrollbox_Items
   widget=FortDynamicEntryBox_Items
   properties {entryWidgetClass=&WBP_UIKit_ItemList_CardBase_C, entrySpacing=<16 0>}
   slot {
    class=ScrollBoxSlot
    properties {horizontalAlignment=HAlign_Left, verticalAlignment=VAlign_Center, padding=<ltrb 0 16 0 16>, size {sizeRule=Fill, value=1}}
   }
  }
  Image_Fade_Left {
   widgetClassPath=&Image
   bIsVariable=true
   parent=Overlay_RowContent
   widget=Image_Fade_Left
   properties {
    visibility=SelfHitTestInvisible
    brush {
     bIsDynamicallyLoaded=false
     drawAs=Image
     imageType=FullColor
     mirroring=NoMirror
     resourceName=None
     resourceObject=/FortUILibrary/Systems/Items/MaterialInstances/MI_UI_ListFade_Left
     tiling=NoTile
     imageSize=<48 32>
     margin=<ltrb 0 0 0 0>
     outlineSettings {
      bUseBrushTransparency=false
      roundingType=HalfHeightRadius
      width=0
      color {colorUseRule=UseColor_Specified, specifiedColor=<rgba 0 0 0 0>}
      cornerRadii=<0 0 0 0>
     }
     tintColor {colorUseRule=UseColor_Specified, specifiedColor=<rgba 1 1 1 1>}
     uVRegion {bIsValid=false, max=<0 0>, min=<0 0>}
    }
   }
   slot {class=GridSlot, properties {horizontalAlignment=HAlign_Left, layer=2}}
  }
  Image_Fade_Right {
   widgetClassPath=&Image
   bIsVariable=true
   parent=Overlay_RowContent
   widget=Image_Fade_Right
   properties {
    visibility=SelfHitTestInvisible
    brush {
     bIsDynamicallyLoaded=false
     drawAs=Image
     imageType=FullColor
     mirroring=NoMirror
     resourceName=None
     resourceObject=/FortUILibrary/Systems/Items/MaterialInstances/MI_UI_ListFade_Right
     tiling=NoTile
     imageSize=<48 32>
     margin=<ltrb 0 0 0 0>
     outlineSettings {
      bUseBrushTransparency=false
      roundingType=HalfHeightRadius
      width=0
      color {colorUseRule=UseColor_Specified, specifiedColor=<rgba 0 0 0 0>}
      cornerRadii=<0 0 0 0>
     }
     tintColor {colorUseRule=UseColor_Specified, specifiedColor=<rgba 1 1 1 1>}
     uVRegion {bIsValid=false, max=<0 0>, min=<0 0>}
    }
   }
   slot {class=GridSlot, properties {horizontalAlignment=HAlign_Right, layer=2}}
  }
  ScrollArrow_Right {
   widgetClassPath=/FortUILibrary/UIKit/Buttons/WBP_UIKit_Button_Quiet
   bIsVariable=true
   parent=Overlay_RowContent
   widget=ScrollArrow_Right
   properties {
    bHideInputAction=true
    bIsFocusable=false
    iconMaterial=/FortUILibrary/Systems/Items/MaterialInstances/MI_UI_ScrollArrow_Right
    text=""
    renderTransformPivot=<0 0.5>
   }
   slot {class=GridSlot, properties {horizontalAlignment=HAlign_Right, layer=4, verticalAlignment=VAlign_Center, nudge=<40 0>}}
  }
  ScrollArrow_Left {
   widgetClassPath=&WBP_UIKit_Button_Quiet
   bIsVariable=true
   parent=Overlay_RowContent
   widget=ScrollArrow_Left
   properties {
    bHideInputAction=true
    bIsFocusable=false
    iconMaterial=/FortUILibrary/Systems/Items/MaterialInstances/MI_UI_ScrollArrow_Left
    minHeight=48
    minWidth=48
    text=""
    renderTransformPivot=<0 0.5>
   }
   slot {class=GridSlot, properties {horizontalAlignment=HAlign_Left, layer=4, verticalAlignment=VAlign_Center, nudge=<-16 0>}}
  }
  Slot_Content {widgetClassPath=&NamedSlot, parent=Overlay_RowContent, widget=Slot_Content, slot {class=GridSlot}}
 ]
}
