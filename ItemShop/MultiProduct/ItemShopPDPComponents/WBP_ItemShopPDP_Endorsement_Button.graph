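# Endorsement button for the Item Shop PDP, extending the generic UIKit button with a check icon that marks a selected preference.
# SetIsPreferenceSelected shows or hides the bottom-right check.
# Hover and unhover drive the background dynamic material's Trans_Hovered parameter and switch the check tint between black and green.
# On Construct, compact UI mode overrides the check and button icon sizes.
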WidgetBlueprint /ItemShop/MultiProduct/ItemShopPDPComponents/WBP_ItemShopPDP_Endorsement_Button
graphs [
 EventGraph {nodes [
   n1 = Event PreConstruct()
   > Parent: PreConstruct(IsDesignTime=@n1.IsDesignTime)
   Event Construct()
   > Parent: Construct()
   > SetCompactUIStyle() {is=title K2Node_CallFunction}
   Event On Hovered()
   > SetBackgroundIsActive!(execute, self, IsActive:bool=true) -> (then)
   > Parent: On Hovered() {is=title K2Node_CallParentFunction}
   > Set ColorAndOpacity!(
    execute
    ColorAndOpacity:Linear Color Structure=(R=0,G=0,B=0,A=1)
    self:Image Object Reference=Get Icon_SelectedCheck() -> (Icon_SelectedCheck:Image Object Reference)
   ) -> (then, Output_Get:Linear Color Structure)
   Event On Unhovered()
   > SetBackgroundIsActive(IsActive=false)
   > Parent: On Unhovered() {is=title K2Node_CallParentFunction}
   > Set ColorAndOpacity(ColorAndOpacity=(R=0.0395,G=1,B=0.2307,A=1), self=Get Icon_SelectedCheck())
   Event Tick()
  ]}
 SetBackgroundIsActive {nodes [
   n1 = SetBackgroundIsActive!() -> (then, IsActive:bool=~) {is=title K2Node_FunctionEntry}
   > n2 = GetBackgroundDynamicMaterial!(execute, self:WBP UIKit Button Generic Base Object Reference) -> (then, Dynamic Material:Material Instance Dynamic Object Reference)
   > SetScalarParameterValue(self=@n2, ParameterName=Trans_Hovered, Value=SelectFloat(A=1, B=0, bPickA=@n1))
  ]}
 SetCompactUIStyle {nodes [
   SetCompactUIStyle() {is=title K2Node_FunctionEntry}
   > Branch(Condition=IsUsingCompactUI())
   > SetDesiredSizeOverride!(execute, self:Image Object Reference=Get Icon_SelectedCheck(), DesiredSize:Vector 2D Structure=(X=24,Y=24)) -> (then)
   > SetIconSize!(execute, self:WBP UIKit Button Generic Base Object Reference, Icon:Vector 2D Structure=(X=40,Y=40)) -> (then)
  ]}
 SetIsPreferenceSelected {nodes [
   n1 = SetIsPreferenceSelected!() -> (then, IsPreferenceSelected:bool=~) {is=title K2Node_FunctionEntry}
   > Set Visibility(Visibility=Select(Option 0=Hidden, Option 1=SelfHitTestInvisible, Index=@n1), self=Get Icon_SelectedCheck())
  ]}
]
variables [
 IconMID {type=MaterialInstanceDynamic}
 IconsRequiringColorChange {type="array<Image>", default []}
 UserPreference {type=EFortItemShopPreferenceLevel, default=None}
]
widgets {
 info {
  inheritedWidgetCount=0
  namedSlotCount=1
  parentClass=/FortUILibrary/UIKit/Buttons/WBP_UIKit_Button_Generic.WBP_UIKit_Button_Generic_C
  rootWidgetClass=null
  widgetCount=2
 }
 widgets [
  Overlay_0 {widgetClassPath=&Overlay, widget=Overlay_0}
  Icon_SelectedCheck {
   widgetClassPath=&Image
   bIsVariable=true
   parent=Overlay_0
   widget=Icon_SelectedCheck
   properties {
    visibility=Hidden
    brush {
     bIsDynamicallyLoaded=false
     drawAs=Image
     imageType=FullColor
     mirroring=NoMirror
     resourceName=None
     resourceObject=/FortUILibrary/Textures/MaterialInstances/MI_UIKit_Icon_Check
     tiling=NoTile
     imageSize=<16 16>
     margin=<ltrb 0 0 0 0>
     outlineSettings {
      bUseBrushTransparency=false
      roundingType=HalfHeightRadius
      width=0
      color {colorUseRule=UseColor_Specified, specifiedColor=<rgba 0 0 0 0>}
      cornerRadii=<0 0 0 0>
     }
     tintColor {colorUseRule=UseColor_Specified, specifiedColor=<rgba 1 1 1 1>}
     uVRegion {bIsValid=false, max=<0 0>, min=<0 0>}
    }
    colorAndOpacity=<rgba 0.0395 1 0.2307 1>
   }
   slot {class=OverlaySlot, properties {horizontalAlignment=HAlign_Right, verticalAlignment=VAlign_Bottom, padding=<ltrb 0 0 7 7>}}
  }
 ]
}
