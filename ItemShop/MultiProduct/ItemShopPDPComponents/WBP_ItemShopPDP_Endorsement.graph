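# Item Shop product detail page widget that lets players rate an offer Love, Like or Dislike from a pop-out drawer.
# A handle switcher toggles the drawer and manages focus; losing focus closes it, and construction shows the widget only if the shop VM allows preferences.
# Clicks write to the offer VM, clearing a repeated choice.
# SetUserPreferenceAppearance updates button highlight and check mark, while the open handle mirrors the current preference's icon.
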WidgetBlueprint /ItemShop/MultiProduct/ItemShopPDPComponents/WBP_ItemShopPDP_Endorsement
graphs [
 EventGraph {nodes [
   "On Clicked (Button_Handle_Open)"!() -> (OutputDelegate, then, Button:Common Button Base Object Reference) {is=title K2Node_ComponentBoundEvent}
   > ToggleDrawer()
   "On Clicked (Button_Handle_Close)"!() -> (OutputDelegate, then, Button:Common Button Base Object Reference) {is=title K2Node_ComponentBoundEvent}
   > ToggleDrawer()
   "On Clicked (Button_Love)"!() -> (OutputDelegate, then, Button:Common Button Base Object Reference) {is=title K2Node_ComponentBoundEvent}
   > SetUserPreferenceOnVM!(execute, self, UserPreference:EFortItemShopPreferenceLevel Enum=StrongPositive) -> (then)
   "On Clicked (Button_Like)"!() -> (OutputDelegate, then, Button:Common Button Base Object Reference) {is=title K2Node_ComponentBoundEvent}
   > SetUserPreferenceOnVM(UserPreference=Positive)
   "On Clicked (Button_Dislike)"!() -> (OutputDelegate, then, Button:Common Button Base Object Reference) {is=title K2Node_ComponentBoundEvent}
   > SetUserPreferenceOnVM(UserPreference=Negative)
   Event Construct()
   n1 = Get bAllowPreferenceBuilder!(self:Fort Item Shop VM Object Reference=Get FortItemShopVM() -> (FortItemShopVM:Fort Item Shop VM Object Reference)) -> (bAllowPreferenceBuilder:bool)
   > SetIsEnabled(bInIsEnabled=@n1)
   > Set Visibility!(execute, Visibility=Select(Option 0=Collapsed, Option 1=SelfHitTestInvisible, Index=@n1)) -> (then, Output_Get)
   > SetCompactUIStyle() {is=title K2Node_CallFunction}
   Event OnRemovedFromFocusPath!() -> (OutputDelegate, then, InFocusEvent:Focus Event Structure) {is=title K2Node_Event}
   > Branch(Condition=IsVisible(self=Get Drawer() -> (Drawer:Common Visual Attachment Object Reference)))
   > ToggleDrawer()
   n2 = Event OnOfferSet!() -> (
    OutputDelegate
    then
    OfferVM:Fort Item Shop Offer VM Object Reference
    ParentOfferVM:Fort Item Shop Offer VM Object Reference
    OfferGroupVM:Fort Item Shop Offer Group VM Object Reference
   ) {is=title K2Node_Event}
   > Set FortItemShopOfferVM(FortItemShopOfferVM:Fort Item Shop Offer VM Object Reference=@n2.OfferVM) -> (Output_Get:Fort Item Shop Offer VM Object Reference)
   Event OnPdpActivated() {is=title K2Node_Event}
   > SetVisibility(self=Get Drawer(), InVisibility=Collapsed)
   > UpdateHandleIcon()
  ]}
 PreferenceToWidget {nodes [Outputs!(Widget:WBP Item Shop PDP Endorsement Button Object Reference=Select!(
     None:WBP Item Shop PDP Endorsement Button Object Reference
     Negative:WBP Item Shop PDP Endorsement Button Object Reference=Get Button_Dislike() -> (Button_Dislike:WBP Item Shop PDP Endorsement Button Object Reference)
     Positive:WBP Item Shop PDP Endorsement Button Object Reference=Get Button_Like() -> (Button_Like:WBP Item Shop PDP Endorsement Button Object Reference)
     StrongPositive:WBP Item Shop PDP Endorsement Button Object Reference=Get Button_Love() -> (Button_Love:WBP Item Shop PDP Endorsement Button Object Reference)
     Index:EFortItemShopPreferenceLevel Enum=Inputs!() -> (UserPreference:EFortItemShopPreferenceLevel Enum=None)
    ) -> (ReturnValue:WBP Item Shop PDP Endorsement Button Object Reference))]}
 SetCompactUIStyle {nodes [
   SetCompactUIStyle() {is=title K2Node_FunctionEntry}
   > Branch(Condition=IsUsingCompactUI())
   > SetDesiredSizeOverride!(
    execute
    self:Image Object Reference=Get Icon_SelectedCheck() -> (Icon_SelectedCheck:Image Object Reference)
    DesiredSize:Vector 2D Structure=(X=24,Y=24)
   ) -> (then)
   > SetDesiredSizeOverride(
    self=[
     Get Handle_Open_Image() -> (Handle_Open_Image:Image Object Reference)
     Get Handle_Close_Image() -> (Handle_Close_Image:Image Object Reference)
    ]
    DesiredSize=(X=40,Y=40)
   )
  ]}
 SetUserPreferenceAppearance {nodes [
   n1 = SetUserPreferenceAppearance!() -> (then, UserPreference:EFortItemShopPreferenceLevel Enum=None) {is=title K2Node_FunctionEntry}
   > n2 = Sequence()
   n3 = Preference to Widget!(UserPreference:EFortItemShopPreferenceLevel Enum=Get CurrentUserPreference() -> (CurrentUserPreference:EFortItemShopPreferenceLevel Enum=None)) -> (Widget:WBP Item Shop PDP Endorsement Button Object Reference) {is=title K2Node_MacroInstance}
   > Is Valid(InputObject=@n3)
   > SetIsPreferenceSelected!(execute, self:WBP Item Shop PDP Endorsement Button Object Reference=@n3, IsPreferenceSelected:bool=false) -> (then)
   n4 = Preference to Widget(UserPreference=@n1)
   n5 = Is Valid(exec=@n2.then_1, InputObject=@n4)
   > SetIsPreferenceSelected(self=@n4, IsPreferenceSelected=true)
   n6 = Get Icon_SelectedCheck()
   > n7 = SetVisibility(self=@n6, InVisibility=SelfHitTestInvisible)
   Set CurrentUserPreference(
    execute=[@n9.then, @n7.then]
    CurrentUserPreference:EFortItemShopPreferenceLevel Enum=Get UserPreference() -> (UserPreference:EFortItemShopPreferenceLevel Enum=None)
   ) -> (Output_Get:EFortItemShopPreferenceLevel Enum=None)
   > n8 = Switch on ESlateVisibility!(
    execute
    Selection:ESlateVisibility Enum=Get Visibility!(self:Widget Object Reference=Get Drawer()) -> (Visibility:ESlateVisibility Enum=Visible)
   ) -> (Visible:Exec, Collapsed:Exec, Hidden:Exec, HitTestInvisible:Exec, SelfHitTestInvisible:Exec) {is=title K2Node_SwitchEnum}
   UpdateHandleIcon(execute=[@n8.Collapsed, @n8.Hidden])
   n9 = SetVisibility(execute=@n5.Is Not Valid, self=@n6, InVisibility=Hidden)
  ]}
 SetUserPreferenceOnVM {nodes [
   n1 = SetUserPreferenceOnVM!() -> (then, UserPreference:EFortItemShopPreferenceLevel Enum=None) {is=title K2Node_FunctionEntry}
   n2 = Get FortItemShopOfferVM() -> (FortItemShopOfferVM:Fort Item Shop Offer VM Object Reference)
   > n3 = Branch(Condition="Equal (Enum)"(
     A:EFortItemShopPreferenceLevel Enum=@n1
     B:EFortItemShopPreferenceLevel Enum=Get UserPreference!(self:Fort Item Shop Offer VM Object Reference=@n2) -> (UserPreference=None)
    ))
   > Set UserPreference!(execute, UserPreference:EFortItemShopPreferenceLevel Enum=None, self:Fort Item Shop Offer VM Object Reference=Get FortItemShopOfferVM()) -> (then, Output_Get:EFortItemShopPreferenceLevel Enum=None)
   Set UserPreference(execute=@n3.else, UserPreference=Get UserPreference!() -> (UserPreference=None), self=Get FortItemShopOfferVM()) -> (Output_Get=None)
  ]}
 ToggleDrawer {nodes [
   ToggleDrawer() {is=title K2Node_FunctionEntry}
   > n1 = Switch on ESlateVisibility(Selection=Get Visibility(self=Get Drawer()) -> (Visibility=Visible))
   SetVisibility(execute=[@n1.Collapsed, @n1.Hidden], self=Get Drawer(), InVisibility=SelfHitTestInvisible)
   > SetActiveWidget(
    self:Widget Switcher Object Reference=Get ButtonHandleSwitcher() -> (ButtonHandleSwitcher:Common Activatable Widget Switcher Object Reference)
    Widget=Get Button_Handle_Close() -> (Button_Handle_Close:WBP UIKit Button Generic Object Reference)
   )
   n2 = Preference to Widget(UserPreference=Get CurrentUserPreference() -> (CurrentUserPreference=None))
   > n3 = Is Valid(InputObject=@n2)
   SetFocus(execute=@n3.Is Not Valid, self=Get Button_Love())
   SetFocus(execute=@n3.Is Valid, self=@n2)
   SetVisibility(execute=[@n1.HitTestInvisible, @n1.SelfHitTestInvisible, @n1.Visible], self=Get Drawer(), InVisibility=Collapsed)
   > UpdateHandleIcon()
   n4 = Get Button_Handle_Open() -> (Button_Handle_Open:WBP UIKit Button Generic Object Reference)
   > SetActiveWidget(self=Get ButtonHandleSwitcher(), Widget=@n4)
   > SetFocus(self=@n4)
  ]}
 UpdateHandleIcon {nodes [
   UpdateHandleIcon() {is=title K2Node_FunctionEntry}
   n1 = Get Button_Dislike()
   n2 = Get Button_Like()
   n3 = Get Button_Love()
   n4 = Get Button_Like()
   > SetBrushFromMaterial(
    self=Get Handle_Open_Image()
    Material=Select(
     None:Material Interface Object Reference=Get IconMaterial!(self:WBP UIKit Button Generic Base Object Reference=@n4) -> (IconMaterial:Material Instance Object Reference)
     Negative:Material Interface Object Reference=Get IconMaterial(self=@n1)
     Positive:Material Interface Object Reference=Get IconMaterial(self=@n2)
     StrongPositive:Material Interface Object Reference=Get IconMaterial(self=@n3)
     Index=Get CurrentUserPreference() -> (CurrentUserPreference=None)
    ) -> (ReturnValue:Material Interface Object Reference)
   )
  ]}
]
variables [
 ActiveButtonBackground {type=MaterialInstance, default=/FortUILibrary/UIKit/Buttons/MaterialInstances/CTAButtons/MI_UIKit_BgRegular}
 CurrentUserPreference {type=EFortItemShopPreferenceLevel, default=None}
 InactiveButtonBackground {type=MaterialInstance, default=/FortUILibrary/UIKit/Buttons/MaterialInstances/CTAButtons/MI_UIKit_IconQuiet}
 InputActionBack {type=DataTableRowHandle, default {dataTable=&InputActionDataTable, rowName=GenericBack}}
 InputActionNone {type=DataTableRowHandle, default {dataTable=null, rowName=None}}
]
widgets {
 info {
  inheritedWidgetCount=0
  namedSlotCount=3
  parentClass=&CommonUserWidget
  rootWidgetClass=&HorizontalBox
  widgetCount=21
 }
 widgets [
  HorizontalBox_34 {widgetClassPath=&HorizontalBox, widget=HorizontalBox_34}
  ButtonHandleSwitcher {
   widgetClassPath=&CommonActivatableWidgetSwitcher
   bIsVariable=true
   parent=HorizontalBox_34
   widget=ButtonHandleSwitcher
   properties {transitionDuration=0}
   slot {class=HorizontalBoxSlot, properties {verticalAlignment=VAlign_Top}}
  }
  Button_Handle_Open {
   widgetClassPath=&WBP_UIKit_Button_Generic
   bIsVariable=true
   parent=ButtonHandleSwitcher
   widget=Button_Handle_Open
   properties {
    bHideInputAction=true
    backgroundMaterial=&MI_UIKit_BgRegular
    inputPriority=1
    mobileMinHeight=90
    mobileMinWidth=90
    outlineMaterial=/FortUILibrary/UIKit/Buttons/MaterialInstances/TileOutline/MI_UIKit_Outline
    priority=1
    text=""
    visibility=Visible
    buttonPadding=<ltrb 28 0 28 0>
   }
   slot {class=WidgetSwitcherSlot}
  }
  Overlay_1 {widgetClassPath=&Overlay, namedSlotHost=Button_Handle_Open, widget=Overlay_1}
  Handle_Open_Image {
   widgetClassPath=&Image
   bIsVariable=true
   parent=Overlay_1
   widget=Handle_Open_Image
   properties {
    visibility=SelfHitTestInvisible
    brush {
     bIsDynamicallyLoaded=false
     drawAs=Image
     imageType=FullColor
     mirroring=NoMirror
     resourceName=None
     resourceObject=/FortUILibrary/Textures/MaterialInstances/MI_UIKit_Icon_Like_Endorsement
     tiling=NoTile
     imageSize=<32 32>
     margin=<ltrb 0 0 0 0>
     outlineSettings {
      bUseBrushTransparency=false
      roundingType=HalfHeightRadius
      width=0
      color {colorUseRule=UseColor_Specified, specifiedColor=<rgba 0 0 0 0>}
      cornerRadii=<0 0 0 0>
     }
     tintColor {colorUseRule=UseColor_Specified, specifiedColor=<rgba 1 1 1 1>}
     uVRegion {bIsValid=false, max=<0 0>, min=<0 0>}
    }
    colorAndOpacity=<rgba 0 0 0 1>
   }
   slot {class=OverlaySlot, properties {horizontalAlignment=HAlign_Center, verticalAlignment=VAlign_Center}}
  }
  Icon_SelectedCheck {
   widgetClassPath=&Image
   bIsVariable=true
   parent=Overlay_1
   widget=Icon_SelectedCheck
   properties {
    visibility=Hidden
    brush {
     bIsDynamicallyLoaded=false
     drawAs=Image
     imageType=FullColor
     mirroring=NoMirror
     resourceName=None
     resourceObject=/FortUILibrary/Textures/MaterialInstances/MI_UIKit_Icon_Check
     tiling=NoTile
     imageSize=<16 16>
     margin=<ltrb 0 0 0 0>
     outlineSettings {
      bUseBrushTransparency=false
      roundingType=HalfHeightRadius
      width=0
      color {colorUseRule=UseColor_Specified, specifiedColor=<rgba 0 0 0 0>}
      cornerRadii=<0 0 0 0>
     }
     tintColor {colorUseRule=UseColor_Specified, specifiedColor=<rgba 1 1 1 1>}
     uVRegion {bIsValid=false, max=<0 0>, min=<0 0>}
    }
    colorAndOpacity=<rgba 0 0 0 1>
   }
   slot {class=OverlaySlot, properties {horizontalAlignment=HAlign_Right, verticalAlignment=VAlign_Bottom, padding=<ltrb 0 0 7 7>}}
  }
  Button_Handle_Close {
   widgetClassPath=&WBP_UIKit_Button_Generic
   bIsVariable=true
   parent=ButtonHandleSwitcher
   widget=Button_Handle_Close
   properties {
    bHideInputAction=true
    backgroundMaterial=&MI_UIKit_BgRegular
    inputPriority=1
    mobileMinHeight=90
    mobileMinWidth=90
    outlineMaterial=&MI_UIKit_Outline
    priority=1
    text=""
    visibility=Visible
    buttonPadding=<ltrb 28 0 28 0>
    triggeringInputAction {dataTable=&InputActionDataTable, rowName=GenericBack}
   }
   slot {class=WidgetSwitcherSlot}
  }
  Overlay {widgetClassPath=&Overlay, namedSlotHost=Button_Handle_Close, widget=Overlay}
  Handle_Close_Image {
   widgetClassPath=&Image
   bIsVariable=true
   parent=Overlay
   widget=Handle_Close_Image
   properties {
    visibility=SelfHitTestInvisible
    brush {
     bIsDynamicallyLoaded=false
     drawAs=Image
     imageType=FullColor
     mirroring=NoMirror
     resourceName=None
     resourceObject=/FortUILibrary/Textures/MaterialInstances/MI_MPLocker_Icon_Clear
     tiling=NoTile
     imageSize=<32 32>
     margin=<ltrb 0 0 0 0>
     outlineSettings {
      bUseBrushTransparency=false
      roundingType=HalfHeightRadius
      width=0
      color {colorUseRule=UseColor_Specified, specifiedColor=<rgba 0 0 0 0>}
      cornerRadii=<0 0 0 0>
     }
     tintColor {colorUseRule=UseColor_Specified, specifiedColor=<rgba 1 1 1 1>}
     uVRegion {bIsValid=false, max=<0 0>, min=<0 0>}
    }
    colorAndOpacity=<rgba 0 0 0 1>
   }
   slot {class=OverlaySlot, properties {horizontalAlignment=HAlign_Center, verticalAlignment=VAlign_Center}}
  }
  Drawer {
   widgetClassPath=&CommonVisualAttachment
   bIsVariable=true
   parent=HorizontalBox_34
   widget=Drawer
   properties {visibility=Collapsed}
   slot {class=HorizontalBoxSlot, properties {horizontalAlignment=HAlign_Left}}
  }
  Overlay_5 {widgetClassPath=&Overlay, parent=Drawer, widget=Overlay_5, slot {class=SizeBoxSlot}}
  Drawer_Background {
   widgetClassPath=&Image
   bIsVariable=true
   parent=Overlay_5
   widget=Drawer_Background
   properties {
    brush {
     bIsDynamicallyLoaded=false
     drawAs=Image
     imageType=FullColor
     mirroring=NoMirror
     resourceName=None
     resourceObject=&MI_UIKit_BgRegular
     tiling=NoTile
     imageSize=<32 32>
     margin=<ltrb 0 0 0 0>
     outlineSettings {
      bUseBrushTransparency=false
      roundingType=HalfHeightRadius
      width=0
      color {colorUseRule=UseColor_Specified, specifiedColor=<rgba 0 0 0 0>}
      cornerRadii=<0 0 0 0>
     }
     tintColor {colorUseRule=UseColor_Specified, specifiedColor=<rgba 1 1 1 1>}
     uVRegion {bIsValid=false, max=<0 0>, min=<0 0>}
    }
    colorAndOpacity=<rgba 0.0048 0.004 0.0116 1>
   }
   slot {class=OverlaySlot, properties {horizontalAlignment=HAlign_Fill, verticalAlignment=VAlign_Fill, padding=<ltrb 16 0 0 0>}}
  }
  Arrow {
   widgetClassPath=&Image
   bIsVariable=true
   parent=Overlay_5
   widget=Arrow
   properties {
    brush {
     bIsDynamicallyLoaded=false
     drawAs=Image
     imageType=FullColor
     mirroring=NoMirror
     resourceName=None
     resourceObject=/Game/Athena/HUD/Compass/M_UI_DiamondMarker
     tiling=NoTile
     imageSize=<32 32>
     margin=<ltrb 0 0 0 0>
     outlineSettings {
      bUseBrushTransparency=false
      roundingType=HalfHeightRadius
      width=0
      color {colorUseRule=UseColor_Specified, specifiedColor=<rgba 0 0 0 0>}
      cornerRadii=<0 0 0 0>
     }
     tintColor {colorUseRule=UseColor_Specified, specifiedColor=<rgba 1 1 1 1>}
     uVRegion {bIsValid=false, max=<0 0>, min=<0 0>}
    }
    colorAndOpacity=<rgba 0.0048 0.004 0.0116 1>
   }
   slot {class=OverlaySlot, properties {verticalAlignment=VAlign_Center}}
  }
  DrawerButtons {
   widgetClassPath=&HorizontalBox
   parent=Overlay_5
   widget=DrawerButtons
   slot {class=OverlaySlot, properties {horizontalAlignment=HAlign_Fill, verticalAlignment=VAlign_Fill, padding=<ltrb 16 0 0 0>}}
  }
  Button_Love {
   widgetClassPath=&/WBP_ItemShopPDP_Endorsement_Button
   bIsVariable=true
   parent=DrawerButtons
   widget=Button_Love
   properties {
    type=Creative
    configuration=Icon
    iconMaterial=/FortUILibrary/Textures/MaterialInstances/MI_UIKit_Icon_Love
    statusIndicatorVerticalAlignment=VAlign_Bottom
    statusIndicatorPadding=<ltrb 0 0 8 8>
   }
   slot {class=HorizontalBoxSlot, properties {padding=<ltrb 0 0 15 0>}}
  }
  Button_Like {
   widgetClassPath=&WBP_ItemShopPDP_Endorsement_Button
   bIsVariable=true
   parent=DrawerButtons
   widget=Button_Like
   slot {class=HorizontalBoxSlot, properties {padding=<ltrb 0 0 15 0>}}
  }
  Button_Dislike {
   widgetClassPath=&WBP_ItemShopPDP_Endorsement_Button
   bIsVariable=true
   parent=DrawerButtons
   widget=Button_Dislike
   properties {iconMaterial=/FortUILibrary/Textures/MaterialInstances/MI_UIKit_Icon_Dislike}
   slot {class=HorizontalBoxSlot}
  }
  Overlay_0 {widgetClassPath=&Overlay, widget=Overlay_0}
  HorizontalBox_92 {
   widgetClassPath=&HorizontalBox
   parent=Overlay_0
   widget=HorizontalBox_92
   slot {class=OverlaySlot, properties {horizontalAlignment=HAlign_Center, verticalAlignment=VAlign_Center, padding=<ltrb 16 0 16 0>}}
  }
  Like {
   widgetClassPath=&Image
   bIsVariable=true
   parent=HorizontalBox_92
   widget=Like
   properties {
    brush {
     bIsDynamicallyLoaded=false
     drawAs=Image
     imageType=FullColor
     mirroring=NoMirror
     resourceName=None
     resourceObject=/FortUILibrary/Textures/MaterialInstances/MI_UIKit_Icon_Like
     tiling=NoTile
     imageSize=<32 32>
     margin=<ltrb 0 0 0 0>
     outlineSettings {
      bUseBrushTransparency=false
      roundingType=HalfHeightRadius
      width=0
      color {colorUseRule=UseColor_Specified, specifiedColor=<rgba 0 0 0 0>}
      cornerRadii=<0 0 0 0>
     }
     tintColor {colorUseRule=UseColor_Specified, specifiedColor=<rgba 1 1 1 1>}
     uVRegion {bIsValid=false, max=<0 0>, min=<0 0>}
    }
    colorAndOpacity=<rgba 0 0 0 1>
   }
   slot {class=HorizontalBoxSlot}
  }
  Dislike {
   widgetClassPath=&Image
   bIsVariable=true
   parent=HorizontalBox_92
   widget=Dislike
   properties {
    brush {
     bIsDynamicallyLoaded=false
     drawAs=Image
     imageType=FullColor
     mirroring=NoMirror
     resourceName=None
     resourceObject=&MI_UIKit_Icon_Like
     tiling=NoTile
     imageSize=<32 32>
     margin=<ltrb 0 0 0 0>
     outlineSettings {
      bUseBrushTransparency=false
      roundingType=HalfHeightRadius
      width=0
      color {colorUseRule=UseColor_Specified, specifiedColor=<rgba 0 0 0 0>}
      cornerRadii=<0 0 0 0>
     }
     tintColor {colorUseRule=UseColor_Specified, specifiedColor=<rgba 1 1 1 1>}
     uVRegion {bIsValid=false, max=<0 0>, min=<0 0>}
    }
    colorAndOpacity=<rgba 0 0 0 1>
    renderTransform {angle=0, scale=<1 -1>, shear=<0 0>, translation=<0 0>}
   }
   slot {class=HorizontalBoxSlot}
  }
 ]
}
