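# Price block for the Item Shop product detail page, showing offer price, owned label, discount disclaimer and real-price text.
# Stores the offer viewmodel when set and broadcasts item-focus changes.
# PreConstruct justifies real-price text and passes disclaimer and optional legal text into the nested price widget.
# Helper functions derive the selected item, hide prices when none remain, and show "Owned" with stack counts for stackable items.
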WidgetBlueprint /ItemShop/MultiProduct/ItemShopPDPComponents/WBP_ItemShopPDP_Price
graphs [
 EventGraph {nodes [
   n1 = Event OnOfferSet!() -> (
    OutputDelegate
    then
    OfferVM:Fort Item Shop Offer VM Object Reference
    ParentOfferVM:Fort Item Shop Offer VM Object Reference
    OfferGroupVM:Fort Item Shop Offer Group VM Object Reference
   ) {is=title K2Node_Event}
   > Set FortItemShopOfferVM(FortItemShopOfferVM:Fort Item Shop Offer VM Object Reference=@n1.OfferVM) -> (Output_Get:Fort Item Shop Offer VM Object Reference)
   Event PreConstruct()
   > Set Justification!(
    execute
    Justification:ETextJustify Enum=Get RealPriceJustification() -> (RealPriceJustification:ETextJustify Enum=Left)
    self:Text Layout Widget Object Reference=Get Text_RealPrice() -> (Text_RealPrice:Common Text Object Reference)
   ) -> (then, Output_Get:ETextJustify Enum=Left)
   > n2 = Get!(execute:Exec) -> (then:Exec, else:Exec, RealPriceLegalText_Optional:Common Text Object Reference)
   n3 = Get WBP_Price_ItemShop() -> (WBP_Price_ItemShop:WBP Price Item Shop Object Reference)
   > Set RealPriceLegalText_Optional!(execute, RealPriceLegalText_Optional:Common Text Object Reference=@n2, self:WBP UIKit Price Object Reference=@n3) -> (then, Output_Get:Common Text Object Reference)
   > Set DiscountDisclaimerText_Optional!(
    execute
    DiscountDisclaimerText_Optional:Common Text Object Reference=Get DiscountDisclaimer() -> (DiscountDisclaimer:Common Text Object Reference)
    self:WBP UIKit Price Object Reference=@n3
   ) -> (then, Output_Get:Common Text Object Reference)
   n4 = Event OnFocusAreaChanged!() -> (OutputDelegate, then, IsFocusingItems:bool) {is=title K2Node_Event}
   > Set with Broadcast IsFocusingItems!(execute, IsFocusingItems:bool=@n4.IsFocusingItems) -> (then, Output_Get:bool)
  ]}
 GetOwnedText {nodes [
   n1 = GetOwnedText!() -> (
    then
    IsOfferOwned:bool=~
    IsFocusingItems:bool=~
    GrantedItems:Array of Fort Item VM Object References
    FocusedItem:Fort Item VM Object Reference
    NumStacksOwned:int
   ) {is=title K2Node_FunctionEntry}
   > n2 = Branch(Condition=@n1)
   n3 = Get FocusedItem() -> (FocusedItem:Fort Item VM Object Reference)
   n4 = "Equal (Integer)"(
    A=Length(TargetArray:Array of Fort Item VM Object References=Get GrantedItems() -> (GrantedItems:Array of Fort Item VM Object References))
    B=1
   )
   n5 = Get IsFocusingItems() -> (IsFocusingItems:bool)
   > n6 = Branch(Condition=AND Boolean(A=OR Boolean(A=@n4, B=@n5), B=IsValid(Object=@n3)))
   n7 = GetMaxNumStacks!(self:Fort Item Definition Object Reference=Get ItemDefinition!(self:Fort Item VM Object Reference=Get FocusedItem()) -> (ItemDefinition:Fort Item Definition Object Reference)) -> (ReturnValue:int=0)
   n8 = Get NumStacksOwned() -> (NumStacksOwned:int=0)
   > n9 = Branch(Condition=AND Boolean(A="integer > integer"(A=@n7, B=1), B="integer > integer"(A=@n8)))
   > Return Node!(
    execute
    Text:Text=Format Text!(Format="LOCTEXT(\"Owned: {numOwned}\")", numOwned:int=Get NumStacksOwned() -> (NumStacksOwned=0)) -> (Result)
   )
   Return Node(execute=[@n9.else, @n6.else], Text="LOCTEXT(\"Owned\")")
   Return Node(execute=@n2.else)
  ]}
 GetPricesVisibility {nodes [
   n1 = GetPricesVisibility!() -> (then, CanPurchaseOffer:EOfferPurchaseError Enum=NoError) {is=title K2Node_FunctionEntry}
   > n2 = Branch(Condition="Not Equal (Enum)"!(A:EOfferPurchaseError Enum=@n1, B:EOfferPurchaseError Enum=NoneLeft) -> (ReturnValue:bool))
   > Return Node!(execute, NewParam:ESlateVisibility Enum=HitTestInvisible)
   Return Node(execute=@n2.else, NewParam=Collapsed)
  ]}
 GetSelectedItem {nodes [
   GetSelectedItem!() -> (then, FocusingItems:bool=~, PreviewedItem:Item Shop Preview Data Object Reference) {is=title K2Node_FunctionEntry}
   > n1 = Get!(execute) -> (then, else, PreviewedItem:Item Shop Preview Data Object Reference)
   > Return Node!(
    execute
    ReturnValue:Fort Item VM Object Reference=Get ItemVM!(self:Item Shop Preview Data Object Reference=@n1) -> (ItemVM:Fort Item VM Object Reference)
   )
   Return Node(
    execute=@n1.else
    ReturnValue=Select(
     Option 0:Fort Item VM Object Reference
     Option 1:Fort Item VM Object Reference=Get FortItemVM() -> (FortItemVM:Fort Item VM Object Reference)
     Index=Get IsFocusingItems()
    ) -> (ReturnValue:Fort Item VM Object Reference)
   )
  ]}
]
variables [
 IsFocusingItems {type=boolean}
 RealPriceJustification {type=ETextJustify, default=Left}
 RealPriceLegalText_Optional {type=CommonTextBlock, default=null}
]
widgets {
 info {inheritedWidgetCount=0, parentClass=&UserWidget, rootWidgetClass=&VerticalBox, widgetCount=7}
 widgets [
  VerticalBox_1 {widgetClassPath=&VerticalBox, widget=VerticalBox_1}
  HorizontalBox_1 {widgetClassPath=&HorizontalBox, parent=VerticalBox_1, widget=HorizontalBox_1, slot {class=VerticalBoxSlot}}
  WBP_Price_ItemShop {
   widgetClassPath=/ItemShop/MultiProduct/WBP_Price_ItemShop
   parent=HorizontalBox_1
   widget=WBP_Price_ItemShop
   properties {realPriceText_Optional=&self:WidgetTree.Text_RealPrice}
   slot {class=HorizontalBoxSlot, properties {padding=<ltrb 0 0 16 0>}}
  }
  Text_Owned {
   widgetClassPath=&CommonTextBlock
   parent=HorizontalBox_1
   widget=Text_Owned
   properties {
    bAutoCollapseWithEmptyText=true
    justification=Center
    mobileFontSizeMultiplier=1.2
    text=Owned
    textTransformPolicy=ToUpper
    colorAndOpacity {colorUseRule=UseColor_Specified, specifiedColor=<rgba 0.0369 1 0.2232 1>}
    font {
     bForceMonospaced=false
     bMaterialIsStencil=false
     fontMaterial=/FortUILibrary/Systems/ItemShopTile/Materials/PriceSwitcher/MI_UIKit_PriceSwitcher_Text
     fontObject=&HeadingNow
     letterSpacing=0
     monospacedWidth=1
     size=24
     skewAmount=0
     typefaceFontName=86 Bold Italic
     outlineSettings {
      bApplyOutlineToDropShadows=false
      bMiteredCorners=false
      bSeparateFillAlpha=false
      outlineMaterial=null
      outlineSize=0
      outlineColor=<rgba 0 0 0 1>
     }
    }
    shadowOffset=<0 0>
   }
   slot {class=HorizontalBoxSlot, properties {horizontalAlignment=HAlign_Left, verticalAlignment=VAlign_Top}}
  }
  DiscountDisclaimer {
   widgetClassPath=&CommonTextBlock
   bIsVariable=true
   parent=VerticalBox_1
   widget=DiscountDisclaimer
   properties {
    applyLineHeightToBottomLine=false
    bAutoCollapseWithEmptyText=true
    justification=Center
    mobileFontSizeMultiplier=1.2
    text="*Discount for select new players; full price applies to all others"
    font {
     bForceMonospaced=false
     bMaterialIsStencil=false
     fontMaterial=null
     fontObject=&HeadingNow
     letterSpacing=0
     monospacedWidth=1
     size=11.25
     skewAmount=0
     typefaceFontName=74 Regular
     outlineSettings {
      bApplyOutlineToDropShadows=false
      bMiteredCorners=false
      bSeparateFillAlpha=false
      outlineMaterial=null
      outlineSize=0
      outlineColor=<rgba 0 0 0 1>
     }
    }
    shadowOffset=<0 0>
   }
   slot {class=VerticalBoxSlot, properties {horizontalAlignment=HAlign_Left, verticalAlignment=VAlign_Top}}
  }
  Visibility_RealPrice {
   widgetClassPath=&Overlay
   parent=VerticalBox_1
   widget=Visibility_RealPrice
   slot {class=VerticalBoxSlot, properties {verticalAlignment=VAlign_Top, padding=<ltrb 0 4 0 0>}}
  }
  Text_RealPrice {
   widgetClassPath=&CommonTextBlock
   bIsVariable=true
   parent=Visibility_RealPrice
   widget=Text_RealPrice
   properties {
    bAutoCollapseWithEmptyText=true
    mobileFontSizeMultiplier=1.2
    text=(equivalent to $8.99*)
    font {
     bForceMonospaced=false
     bMaterialIsStencil=false
     fontMaterial=null
     fontObject=&HeadingNow
     letterSpacing=0
     monospacedWidth=1
     size=18
     skewAmount=0
     typefaceFontName=74 Regular
     outlineSettings {
      bApplyOutlineToDropShadows=false
      bMiteredCorners=false
      bSeparateFillAlpha=false
      outlineMaterial=null
      outlineSize=0
      outlineColor=<rgba 0 0 0 1>
     }
    }
    shadowOffset=<0 0>
   }
   slot {class=OverlaySlot, properties {horizontalAlignment=HAlign_Fill}}
  }
 ]
}
