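# Item shop coach mark widget (ALC variant, under MultiProduct) derived from the UIKit CoachMark widget.
# It adds no widgets or variables of its own and relies on the parent class for layout.
# Its only logic overrides OnButtonDismissClicked, calling the parent handler and then deactivating the widget.
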WidgetBlueprint /ItemShop/MultiProduct/WBP_ItemShop_Coachmark_ALC
graphs [EventGraph {nodes [
   n1 = Event OnButtonDismissClicked!() -> (OutputDelegate, then, Button:Common Button Base Object Reference) {is=title K2Node_Event}
   > Parent: OnButtonDismissClicked!(execute, Button:Common Button Base Object Reference=@n1.Button) -> (then) {is=title K2Node_CallParentFunction}
   > DeactivateWidget()
  ]}]
variables []
widgets {
 info {
  inheritedWidgetCount=0
  parentClass=/FortUILibrary/UIKit/CoachMark/WBP_UIKit_CoachMark.WBP_UIKit_CoachMark_C
  rootWidgetClass=null
  widgetCount=0
 }
 widgets []
}
