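# Item card widget for multi-product Item Shop offers, derived from the UIKit item card base with an added count label and owned check indicator.
# SetCount swaps the item image for a "+N" label when Count exceeds 1, otherwise it shows the image.
# PreConstruct collapses the block and equipped outlines and turns off the enhanced look.
# SetIsInteractable toggles the block outline, hit-testing and selectability.
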WidgetBlueprint /ItemShop/MultiProduct/WBP_ItemShop_ItemCard_ALC
graphs [
 EventGraph {nodes [
   n1 = SetCount!() -> (OutputDelegate, then, Count:int) {is=title K2Node_CustomEvent}
   > n2 = Branch(Condition="integer > integer"(A=@n1.Count, B=1))
   > Set Visibility(Visibility=SelfHitTestInvisible, self=Get Text_Count() -> (Text_Count:Common Text Object Reference))
   > Set Text(Text=Format Text!(Format="LOCTEXT(\"+{Count}\")", Count:int=@n1.Count) -> (Result), self=Get Text_Count())
   > Set Visibility(Visibility=Collapsed, self=Get LazyImage_Item() -> (LazyImage_Item:Fort Lazy Image Object Reference))
   Set Visibility(execute=@n2.else, Visibility=Collapsed, self=Get Text_Count())
   > Set Visibility(Visibility=SelfHitTestInvisible, self=Get LazyImage_Item())
   n3 = Event PreConstruct()
   > Parent: PreConstruct(IsDesignTime=@n3.IsDesignTime)
   > SetVisibility(
    self=[
     Get Block_Outline() -> (Block_Outline:WBP UIKit Block Outline Object Reference)
     Get EquippedState_Outline() -> (EquippedState_Outline:WBP UIKit Equipped State Object Reference)
    ]
    InVisibility=Collapsed
   )
   > SetItemIsEnhanced!(execute, self:WBP UIKit Item Card Base Object Reference, Is Enhanced:bool=false) -> (then)
  ]}
 SetIsInteractable {nodes [
   n1 = SetIsInteractable!() -> (then, IsInteractable:bool=~) {is=title K2Node_FunctionEntry}
   > SetVisibility(self=Get Block_Outline(), InVisibility=Select(Option 0=Collapsed, Option 1=SelfHitTestInvisible, Index=@n1))
   > SetVisibility(InVisibility=Select(Option 0=HitTestInvisible, Option 1=Visible, Index=Get IsInteractable() -> (IsInteractable:bool)))
   > SetIsSelectable!(execute, self:Common Button Base Object Reference, bInIsSelectable:bool=Get IsInteractable()) -> (then)
  ]}
]
variables []
widgets {
 info {
  inheritedWidgetCount=0
  namedSlotCount=1
  parentClass=/FortUILibrary/Systems/Items/WBP_UIKit_ItemCard_Base.WBP_UIKit_ItemCard_Base_C
  rootWidgetClass=null
  widgetCount=3
 }
 widgets [
  Overlay_0 {widgetClassPath=&Overlay, widget=Overlay_0}
  Text_Count {
   widgetClassPath=&CommonTextBlock
   bIsVariable=true
   parent=Overlay_0
   widget=Text_Count
   properties {
    mobileFontSizeMultiplier=1.2
    text=+2
    visibility=Collapsed
    font {
     bForceMonospaced=false
     bMaterialIsStencil=false
     fontMaterial=null
     fontObject=&HeadingNow
     letterSpacing=0
     monospacedWidth=1
     size=24
     skewAmount=0
     typefaceFontName=74 Regular
     outlineSettings {
      bApplyOutlineToDropShadows=false
      bMiteredCorners=false
      bSeparateFillAlpha=false
      outlineMaterial=null
      outlineSize=0
      outlineColor=<rgba 0 0 0 1>
     }
    }
    shadowOffset=<0 0>
   }
   slot {class=OverlaySlot, properties {horizontalAlignment=HAlign_Center, verticalAlignment=VAlign_Center}}
  }
  StatusIndicator_Owned {
   widgetClassPath=/FortUILibrary/UIKit/StatusIndicator/WBP_UIKit_StatusIndicator
   parent=Overlay_0
   widget=StatusIndicator_Owned
   properties {icon=/FortUILibrary/Textures/MaterialInstances/MI_UIKit_Icon_Check, iconSize=Compact, overrideStatusIndicatorVisibility=true}
   slot {class=OverlaySlot, properties {horizontalAlignment=HAlign_Right, padding=<ltrb 4 4 4 4>}}
  }
 ]
}
