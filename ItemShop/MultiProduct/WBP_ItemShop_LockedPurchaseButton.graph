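# Item shop widget showing a disabled, locked purchase button that states how much XP unlocks the ability to purchase.
# PreConstruct locks the wrapped generic UIKit button, enables scrolling text, raises minimum height on mobile, and fills a preview XP value at design time.
# SetButtonText formats the localized message with the supplied XP, showing plain integers below 999 and shorthand quantities otherwise.
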WidgetBlueprint /ItemShop/MultiProduct/WBP_ItemShop_LockedPurchaseButton
graphs [
 EventGraph {nodes [
   n1 = Event PreConstruct()
   > SetIsLocked!(
    execute
    self:Common Button Base Object Reference=Get WBP_UIKit_Button_Generic() -> (WBP_UIKit_Button_Generic:WBP UIKit Button Generic Object Reference)
    bInIsLocked:bool=true
   ) -> (then)
   n2 = Get WBP_UIKit_Button_Generic()
   > SetScrollingEnabled!(
    execute
    self:Common Text Object Reference=Get TextBlock!(self:WBP UIKit Block Text Object Reference=Get Block_Text!(self:WBP UIKit Button Generic Base Object Reference=@n2) -> (Block_Text:WBP UIKit Block Text Object Reference)) -> (TextBlock:Common Text Object Reference)
    bInIsScrollingEnabled:bool=true
   ) -> (then)
   > SetMobileStyle() {is=title K2Node_CallFunction}
   > Branch(Condition=@n1.IsDesignTime)
   > SetButtonText!(execute, self, RequiredXP:int=Get PreviewRequiredXP() -> (PreviewRequiredXP:int=0)) -> (then)
  ]}
 SetButtonText {nodes [
   SetButtonText!() -> (then, RequiredXP:int) {is=title K2Node_FunctionEntry}
   n1 = Get RequiredXP() -> (RequiredXP:int=0)
   n2 = GetShorthandQuantity!(Quantity:int=@n1, MinimumFractionalDigits:int=0, MaximumFractionalDigits:int=2, bShouldRoundDown:bool=false) -> (ReturnValue:Text)
   n3 = "To Text (Integer)"(Value=@n1)
   n4 = "integer < integer"(A=@n1, B=999)
   > SetText!(
    execute
    self:WBP UIKit Button Generic Base Object Reference=Get WBP_UIKit_Button_Generic()
    Text:Text=Format Text!(Format="LOCTEXT(\"{requiredXP} XP unlocks ability to purchase\")", requiredXP:Text=SelectText(A=@n3, B=@n2, bPickA=@n4)) -> (Result)
   ) -> (then)
  ]}
 SetMobileStyle {nodes [
   SetMobileStyle() {is=title K2Node_FunctionEntry}
   > Branch(Condition=IsMobileGame())
   > SetMinDesiredHeight!(execute, self:Size Box Object Reference=Get SizeBox() -> (SizeBox:Size Box Object Reference), InMinDesiredHeight:float=80) -> (then)
  ]}
]
variables [PreviewRequiredXP {type=integer, default=10500}]
widgets {
 info {inheritedWidgetCount=0, parentClass=&UserWidget, rootWidgetClass=&SizeBox, widgetCount=2}
 widgets [
  SizeBox {
   widgetClassPath=&SizeBox
   bIsVariable=true
   widget=SizeBox
   properties {bOverride_MinDesiredHeight=true, bOverride_WidthOverride=true, minDesiredHeight=64, widthOverride=624}
  }
  WBP_UIKit_Button_Generic {
   widgetClassPath=&WBP_UIKit_Button_Generic
   bIsVariable=true
   parent=SizeBox
   widget=WBP_UIKit_Button_Generic
   properties {
    bIsEnabled=false
    bIsFocusable=false
    bLocked=true
    icon=/FortUILibrary/Textures/Icons/ToBeDeprecated/T_UI_Icon_Global_Locked
    iconMaterial=/FortUILibrary/UIKit/Buttons/MaterialInstances/GenericButtons/MI_UIKit_GenericIcon_Regular
    maxWidth=620
    minHeight=64
    minWidth=620
    text="[PH] 100K XP UNLOCKS ABILITY TO PURCHASE"
    textOverflowPolicy=Ellipsis
    useDisabledVisualsOnLocked=true
    font {
     bForceMonospaced=false
     bMaterialIsStencil=false
     fontMaterial=/FortUILibrary/UIKit/Buttons/MaterialInstances/CTAButtons/MI_UIKit_TextRegular
     fontObject=&HeadingNow
     letterSpacing=48
     monospacedWidth=1
     size=15.75
     skewAmount=0
     typefaceFontName=76 Bold
     outlineSettings {
      bApplyOutlineToDropShadows=false
      bMiteredCorners=false
      bSeparateFillAlpha=false
      outlineMaterial=null
      outlineSize=0
      outlineColor=<rgba 0 0 0 1>
     }
    }
    iconPadding=<ltrb 0 0 12 0>
    iconSize=<32 32>
   }
   slot {class=SizeBoxSlot}
  }
 ]
}
