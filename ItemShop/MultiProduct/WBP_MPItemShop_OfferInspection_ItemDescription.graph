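# Item description panel for the multi-product Item Shop offer inspection screen, derived from WBP_UIKit_DescriptionWithSubtitle with no widgets of its own.
# ChangeView adapts the inherited layout to the product detail page view enum.
# OfferInspect removes the body's top padding and shows the bottom named slot, while ItemDetails keeps the configured top padding and collapses that slot.
# PreConstruct calls the parent, then applies OfferInspect as the initial view.
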WidgetBlueprint /ItemShop/MultiProduct/WBP_MPItemShop_OfferInspection_ItemDescription
graphs [
 ChangeView {nodes [
   n1 = ChangeView!() -> (then, NewView:EFortItemShopPdpView Enum=OfferInspect) {is=title K2Node_FunctionEntry}
   n2 = Get!() -> (BodyTextPadding_Left:float=~, BodyTextPadding_Top:float=~, BodyTextPadding_Right:float=~, BodyTextPadding_Bottom:float=~)
   > SetGridSlotPadding!(
    execute
    Widget:Widget Object Reference=Get VBox_Body() -> (VBox_Body:Vertical Box Object Reference)
    Padding_Left:float=@n2
    Padding_Top:float=Select!(OfferInspect:float=0, ItemDetails:float=@n2.BodyTextPadding_Top, Index:EFortItemShopPdpView Enum=@n1) -> (ReturnValue:float)
    Padding_Right:float=@n2.BodyTextPadding_Right
    Padding_Bottom:float=@n2.BodyTextPadding_Bottom
   ) -> (then)
   > Set Visibility(
    Visibility=Select(OfferInspect:ESlateVisibility Enum=SelfHitTestInvisible, ItemDetails:ESlateVisibility Enum=Collapsed, Index=@n1) -> (ReturnValue:ESlateVisibility Enum)
    self=Get Slot_Bottom() -> (Slot_Bottom:Named Slot Object Reference)
   )
  ]}
 EventGraph {nodes [
   n1 = Event PreConstruct()
   > Parent: PreConstruct(IsDesignTime=@n1.IsDesignTime)
   > ChangeView!(execute, self, NewView:EFortItemShopPdpView Enum=OfferInspect) -> (then) {is=title K2Node_CallFunction}
  ]}
]
variables []
widgets {
 info {
  inheritedWidgetCount=0
  parentClass=/FortUILibrary/UIKit/Items/WBP_UIKit_DescriptionWithSubtitle.WBP_UIKit_DescriptionWithSubtitle_C
  rootWidgetClass=null
  widgetCount=0
 }
 widgets []
}
